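# Physics asset for the DangerMojo female medium body, covering the jacket, hood, hood straps and obi belt dynamic bones.
# Capsules on pelvis, spine and clavicles form the torso chain.
# Six jacket panels hang from spine_04, the belt from spine_02, hood sides from clavicles, and two-segment straps from spine_05.
# Torso, base and hood joints swing and twist freely, while jacket, belt and strap segments have angular limits with tight twist.
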PhysicsAsset /BRCosmetics/Characters/Player/Female/Medium/Bodies/F_MED_DangerMojo/Meshes/F_MED_DangerMojo_Physics_jaket
bodies {
 clavicle_l [{shapeType=Capsule, length=10, radius=5, center=<5.3912 0 -2.0329>, rotation=<rot -90 180 0>}]
 clavicle_r [{shapeType=Capsule, length=10, radius=5, center=<-5.3912 0 2.0329>, rotation=<rot -90 0 0>}]
 dyn_hood_l [{shapeType=Capsule, length=1, radius=1}]
 dyn_hood_l_base [{shapeType=Capsule, length=1, radius=1}]
 dyn_hood_r [{shapeType=Capsule, length=1, radius=1}]
 dyn_hood_r_base [{shapeType=Capsule, length=1, radius=1}]
 dyn_hood_strap_l_01 [{shapeType=Capsule, length=3, radius=1, center=<3 0 0>, rotation=<rot 90 0 0>}]
 dyn_hood_strap_l_02 [{shapeType=Capsule, length=3, radius=1, center=<3 0 0>, rotation=<rot 90 0 0>}]
 dyn_hood_strap_l_base [{shapeType=Capsule, length=5.5, radius=5.5, center=<4.9695 -0.889 -6.0993>, rotation=<rot 0 0 90>}]
 dyn_hood_strap_r_01 [{shapeType=Capsule, length=3, radius=1, center=<3 0 0>, rotation=<rot 90 0 0>}]
 dyn_hood_strap_r_02 [{shapeType=Capsule, length=3, radius=1, center=<3 0 0>, rotation=<rot 90 0 0>}]
 dyn_hood_strap_r_base [{shapeType=Capsule, length=5.5, radius=5.5, center=<4.9695 0.889 -6.0993>, rotation=<rot 0 0 -90>}]
 dyn_jaket_bk_centr [{shapeType=Capsule, length=5, radius=5, center=<8 0 0>, rotation=<rot 0 0 90>}]
 dyn_jaket_bk_centr_base [{shapeType=Capsule, length=2, radius=2, rotation=<rot 0 0 90>}]
 dyn_jaket_bk_side_l [{shapeType=Capsule, length=5, radius=3, center=<8 0 0>, rotation=<rot 0 0 90>}]
 dyn_jaket_bk_side_l_base [{shapeType=Capsule, length=2, radius=2, rotation=<rot 0 0 90>}]
 dyn_jaket_bk_side_r [{shapeType=Capsule, length=5, radius=3, center=<8 0 0>, rotation=<rot 0 0 90>}]
 dyn_jaket_bk_side_r_base [{shapeType=Capsule, length=2, radius=2, rotation=<rot 0 0 90>}]
 dyn_jaket_fr_l [{shapeType=Capsule, length=5, radius=3, center=<8 0 0>, rotation=<rot 0 0 90>}]
 dyn_jaket_fr_l_base [{shapeType=Capsule, length=2, radius=2, rotation=<rot 0 0 90>}]
 dyn_jaket_fr_r [{shapeType=Capsule, length=5, radius=3, center=<8 0 0>, rotation=<rot 0 0 90>}]
 dyn_jaket_fr_r_base [{shapeType=Capsule, length=2, radius=2, rotation=<rot 0 0 90>}]
 dyn_jaket_fr_side_l [{shapeType=Capsule, length=5, radius=3, center=<8 0 0>, rotation=<rot 0 0 90>}]
 dyn_jaket_fr_side_l_base [{shapeType=Capsule, length=2, radius=2, rotation=<rot 0 0 90>}]
 dyn_jaket_fr_side_r [{shapeType=Capsule, length=5, radius=3, center=<8 0 0>, rotation=<rot 0 0 90>}]
 dyn_jaket_fr_side_r_base [{shapeType=Capsule, length=2, radius=2, rotation=<rot 0 0 90>}]
 dyn_obi_belt_up [{shapeType=Capsule, length=9.0368, radius=2.5, shapeName=dyn_obi_belt_up_capsule, center=<6.797 0.0006 -1.4858>, rotation=<rot 0 0 -89.9999>}]
 dyn_obi_belt_up_base [{shapeType=Capsule, length=0.505, radius=0.505, shapeName=dyn_obi_belt_up_base_capsule, rotation=<rot 0 -90 0>}]
 pelvis [{shapeType=Capsule, length=10, radius=8}]
 spine_01 [{shapeType=Capsule, length=10, radius=8}]
 spine_02 [{shapeType=Capsule, length=7, radius=7}, {shapeType=Capsule, length=7, radius=7, shapeName=spine_02_capsule, center=<3.5772 5.4983 0>}]
 spine_03 [{shapeType=Capsule, length=7, radius=7}]
 spine_04 [{shapeType=Capsule, length=14.5, radius=8, center=<1.8049 -23.124 0>}]
 spine_05 [{shapeType=Capsule, length=14, radius=8, center=<0 1.1039 0>}]
}
constraints [
 {
  bone1Name=spine_01
  bone2Name=pelvis
  swing1LimitDegrees=45
  swing1Motion=Free
  swing2LimitDegrees=45
  swing2Motion=Free
  twistLimitDegrees=45
  twistMotion=Free
 }
 {
  bone1Name=spine_02
  bone2Name=spine_01
  swing1LimitDegrees=45
  swing1Motion=Free
  swing2LimitDegrees=45
  swing2Motion=Free
  twistLimitDegrees=45
  twistMotion=Free
 }
 {
  bone1Name=spine_03
  bone2Name=spine_02
  swing1LimitDegrees=45
  swing1Motion=Free
  swing2LimitDegrees=45
  swing2Motion=Free
  twistLimitDegrees=45
  twistMotion=Free
 }
 {
  bone1Name=spine_04
  bone2Name=spine_03
  swing1LimitDegrees=45
  swing1Motion=Free
  swing2LimitDegrees=45
  swing2Motion=Free
  twistLimitDegrees=45
  twistMotion=Free
 }
 {
  bone1Name=dyn_jaket_bk_centr_base
  bone2Name=spine_04
  swing1LimitDegrees=45
  swing1Motion=Free
  swing2LimitDegrees=45
  swing2Motion=Free
  twistLimitDegrees=45
  twistMotion=Free
 }
 {bone1Name=dyn_jaket_bk_centr, bone2Name=dyn_jaket_bk_centr_base, swing1LimitDegrees=15, swing2LimitDegrees=30, twistLimitDegrees=3}
 {
  bone1Name=dyn_jaket_bk_side_l_base
  bone2Name=spine_04
  swing1LimitDegrees=45
  swing1Motion=Free
  swing2LimitDegrees=45
  swing2Motion=Free
  twistLimitDegrees=45
  twistMotion=Free
 }
 {bone1Name=dyn_jaket_bk_side_l, bone2Name=dyn_jaket_bk_side_l_base, swing1LimitDegrees=15, swing2LimitDegrees=45, twistLimitDegrees=3}
 {
  bone1Name=dyn_jaket_bk_side_r_base
  bone2Name=spine_04
  swing1LimitDegrees=45
  swing1Motion=Free
  swing2LimitDegrees=45
  swing2Motion=Free
  twistLimitDegrees=45
  twistMotion=Free
 }
 {bone1Name=dyn_jaket_bk_side_r, bone2Name=dyn_jaket_bk_side_r_base, swing1LimitDegrees=15, swing2LimitDegrees=45, twistLimitDegrees=3}
 {
  bone1Name=dyn_jaket_fr_l_base
  bone2Name=spine_04
  swing1LimitDegrees=45
  swing1Motion=Free
  swing2LimitDegrees=45
  swing2Motion=Free
  twistLimitDegrees=45
  twistMotion=Free
 }
 {bone1Name=dyn_jaket_fr_l, bone2Name=dyn_jaket_fr_l_base, swing1LimitDegrees=15, swing2LimitDegrees=45, twistLimitDegrees=3}
 {
  bone1Name=dyn_jaket_fr_r_base
  bone2Name=spine_04
  swing1LimitDegrees=45
  swing1Motion=Free
  swing2LimitDegrees=45
  swing2Motion=Free
  twistLimitDegrees=45
  twistMotion=Free
 }
 {bone1Name=dyn_jaket_fr_r, bone2Name=dyn_jaket_fr_r_base, swing1LimitDegrees=15, swing2LimitDegrees=45, twistLimitDegrees=3}
 {
  bone1Name=dyn_jaket_fr_side_l_base
  bone2Name=spine_04
  swing1LimitDegrees=45
  swing1Motion=Free
  swing2LimitDegrees=45
  swing2Motion=Free
  twistLimitDegrees=45
  twistMotion=Free
 }
 {bone1Name=dyn_jaket_fr_side_l, bone2Name=dyn_jaket_fr_side_l_base, swing1LimitDegrees=15, swing2LimitDegrees=45, twistLimitDegrees=3}
 {
  bone1Name=dyn_jaket_fr_side_r_base
  bone2Name=spine_04
  swing1LimitDegrees=45
  swing1Motion=Free
  swing2LimitDegrees=45
  swing2Motion=Free
  twistLimitDegrees=45
  twistMotion=Free
 }
 {bone1Name=dyn_jaket_fr_side_r, bone2Name=dyn_jaket_fr_side_r_base, swing1LimitDegrees=15, swing2LimitDegrees=45, twistLimitDegrees=3}
 {
  bone1Name=dyn_obi_belt_up_base
  bone2Name=spine_02
  swing1LimitDegrees=45
  swing1Motion=Free
  swing2LimitDegrees=45
  swing2Motion=Free
  twistLimitDegrees=45
  twistMotion=Free
 }
 {bone1Name=dyn_obi_belt_up, bone2Name=dyn_obi_belt_up_base, swing1LimitDegrees=10, swing2LimitDegrees=30, twistLimitDegrees=5}
 {
  bone1Name=dyn_hood_l_base
  bone2Name=clavicle_l
  swing1LimitDegrees=45
  swing1Motion=Free
  swing2LimitDegrees=45
  swing2Motion=Free
  twistLimitDegrees=45
  twistMotion=Free
 }
 {
  bone1Name=dyn_hood_l
  bone2Name=dyn_hood_l_base
  swing1LimitDegrees=45
  swing1Motion=Free
  swing2LimitDegrees=45
  swing2Motion=Free
  twistLimitDegrees=45
  twistMotion=Free
 }
 {
  bone1Name=dyn_hood_r_base
  bone2Name=clavicle_r
  swing1LimitDegrees=45
  swing1Motion=Free
  swing2LimitDegrees=45
  swing2Motion=Free
  twistLimitDegrees=45
  twistMotion=Free
 }
 {
  bone1Name=dyn_hood_r
  bone2Name=dyn_hood_r_base
  swing1LimitDegrees=45
  swing1Motion=Free
  swing2LimitDegrees=45
  swing2Motion=Free
  twistLimitDegrees=45
  twistMotion=Free
 }
 {
  bone1Name=dyn_hood_strap_l_base
  bone2Name=spine_05
  swing1LimitDegrees=45
  swing1Motion=Free
  swing2LimitDegrees=45
  swing2Motion=Free
  twistLimitDegrees=45
  twistMotion=Free
 }
 {bone1Name=dyn_hood_strap_l_01, bone2Name=dyn_hood_strap_l_base, swing1LimitDegrees=30, swing2LimitDegrees=30, twistLimitDegrees=5}
 {bone1Name=dyn_hood_strap_l_02, bone2Name=dyn_hood_strap_l_01, swing1LimitDegrees=30, swing2LimitDegrees=30, twistLimitDegrees=5}
 {bone1Name=dyn_hood_strap_r_01, bone2Name=dyn_hood_strap_r_base, swing1LimitDegrees=30, swing2LimitDegrees=30, twistLimitDegrees=5}
 {bone1Name=dyn_hood_strap_r_02, bone2Name=dyn_hood_strap_r_01, swing1LimitDegrees=30, swing2LimitDegrees=30, twistLimitDegrees=5}
 {
  bone1Name=spine_05
  bone2Name=spine_04
  swing1LimitDegrees=45
  swing1Motion=Free
  swing2LimitDegrees=45
  swing2Motion=Free
  twistLimitDegrees=45
  twistMotion=Free
 }
 {
  bone1Name=clavicle_r
  bone2Name=spine_05
  swing1LimitDegrees=45
  swing1Motion=Free
  swing2LimitDegrees=45
  swing2Motion=Free
  twistLimitDegrees=45
  twistMotion=Free
 }
 {
  bone1Name=clavicle_l
  bone2Name=spine_05
  swing1LimitDegrees=45
  swing1Motion=Free
  swing2LimitDegrees=45
  swing2Motion=Free
  twistLimitDegrees=45
  twistMotion=Free
 }
 {
  bone1Name=dyn_hood_strap_r_base
  bone2Name=spine_05
  swing1LimitDegrees=45
  swing1Motion=Free
  swing2LimitDegrees=45
  swing2Motion=Free
  twistLimitDegrees=45
  twistMotion=Free
 }
]
