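# Physics asset that drives secondary motion for the skirt and obi belt of the female medium DangerMojo body.
# Five three-segment capsule chains (back center, back left/right, both sides) hang from base bodies linked to spine_01 with free angular motion, using tight swing and twist limits along each chain.
# The belt's base links to the first back-center segment, while pelvis, spine and thigh capsules supply body volumes.
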PhysicsAsset /BRCosmetics/Characters/Player/Female/Medium/Bodies/F_MED_DangerMojo/Meshes/F_MED_DangerMojo_Physics_skirt
bodies {
 dyn_obi_belt_down [{shapeType=Capsule, length=6, radius=4, shapeName=dyn_obi_belt_down_capsule, center=<6 0 0>, rotation=<rot 90 0 0>}]
 dyn_obi_belt_down_base [{shapeType=Capsule, length=0.505, radius=0.505, shapeName=dyn_obi_belt_down_base_capsule, rotation=<rot 0 -90 0>}]
 dyn_skirt_bk_centr_01 [{shapeType=Capsule, length=8, radius=2.5, center=<4.7075 0 1.5194>, rotation=<rot 0 0 90>}]
 dyn_skirt_bk_centr_02 [{shapeType=Capsule, length=10, radius=3, center=<5 0 0>, rotation=<rot 0 0 90>}]
 dyn_skirt_bk_centr_03 [{shapeType=Capsule, length=10, radius=4, center=<4 0 0>, rotation=<rot 0 0 90>}]
 dyn_skirt_bk_centr_base [{shapeType=Capsule, length=1, radius=1}]
 dyn_skirt_bk_l_01 [{shapeType=Capsule, length=8, radius=2.5, center=<4 -2 0>, rotation=<rot 0 0 90>}]
 dyn_skirt_bk_l_02 [{shapeType=Capsule, length=10, radius=3, center=<5 0 0>, rotation=<rot 0 0 90>}]
 dyn_skirt_bk_l_03 [{shapeType=Capsule, length=10, radius=4, center=<4 0 0>, rotation=<rot 0 0 90>}]
 dyn_skirt_bk_l_base [{shapeType=Capsule, length=1, radius=1}]
 dyn_skirt_bk_r_01 [{shapeType=Capsule, length=8, radius=2.5, center=<4 2 0>, rotation=<rot 0 0 90>}]
 dyn_skirt_bk_r_02 [{shapeType=Capsule, length=10, radius=3, center=<5 0 0>, rotation=<rot 0 0 90>}]
 dyn_skirt_bk_r_03 [{shapeType=Capsule, length=10, radius=4, center=<4 0 0>, rotation=<rot 0 0 90>}]
 dyn_skirt_bk_r_base [{shapeType=Capsule, length=1, radius=1}]
 dyn_skirt_side_l_01 [{shapeType=Capsule, length=8, radius=2, center=<4 0 0>, rotation=<rot 0 0 90>}]
 dyn_skirt_side_l_02 [{shapeType=Capsule, length=10, radius=3, center=<5 0 0>, rotation=<rot 0 0 90>}]
 dyn_skirt_side_l_03 [{shapeType=Capsule, length=10, radius=4, center=<4 0 0>, rotation=<rot 0 0 90>}]
 dyn_skirt_side_l_base [{shapeType=Capsule, length=1, radius=1}]
 dyn_skirt_side_r_01 [{shapeType=Capsule, length=8, radius=3, center=<4 0 0>, rotation=<rot 0 0 90>}]
 dyn_skirt_side_r_02 [{shapeType=Capsule, length=10, radius=3, center=<5 0 0>, rotation=<rot 0 0 90>}]
 dyn_skirt_side_r_03 [{shapeType=Capsule, length=10, radius=4, center=<4 0 0>, rotation=<rot 0 0 90>}]
 dyn_skirt_side_r_base [{shapeType=Capsule, length=1, radius=1}]
 pelvis [
  {shapeType=Capsule, length=6, radius=9, center=<-1.7446 2.3079 0>}
  {shapeType=Capsule, length=8, radius=9, shapeName=pelvis_capsule, center=<-0.6396 0.1353 0>}
 ]
 spine_01 [{shapeType=Capsule, length=6, radius=8, center=<0 1.4539 0>}]
 thigh_l [{shapeType=Capsule, length=20, radius=8, center=<-17.2057 -0.0085 1.4762>, rotation=<rot 0 -87.6286 -97.9391>}]
 thigh_r [{shapeType=Capsule, length=20, radius=8, center=<17.2057 0.0085 -1.4762>, rotation=<rot 0 -87.6286 82.0609>}]
}
constraints [
 {bone1Name=spine_01, bone2Name=pelvis, swing1LimitDegrees=45, swing2LimitDegrees=45, twistLimitDegrees=45}
 {bone1Name=thigh_l, bone2Name=pelvis, swing1LimitDegrees=45, swing2LimitDegrees=45, twistLimitDegrees=45}
 {
  bone1Name=dyn_skirt_bk_centr_base
  bone2Name=spine_01
  swing1LimitDegrees=45
  swing1Motion=Free
  swing2LimitDegrees=45
  swing2Motion=Free
  twistLimitDegrees=45
  twistMotion=Free
 }
 {
  bone1Name=dyn_skirt_bk_l_base
  bone2Name=spine_01
  swing1LimitDegrees=45
  swing1Motion=Free
  swing2LimitDegrees=45
  swing2Motion=Free
  twistLimitDegrees=45
  twistMotion=Free
 }
 {
  bone1Name=dyn_skirt_bk_r_base
  bone2Name=spine_01
  swing1LimitDegrees=45
  swing1Motion=Free
  swing2LimitDegrees=45
  swing2Motion=Free
  twistLimitDegrees=45
  twistMotion=Free
 }
 {
  bone1Name=dyn_skirt_side_l_base
  bone2Name=spine_01
  swing1LimitDegrees=45
  swing1Motion=Free
  swing2LimitDegrees=45
  swing2Motion=Free
  twistLimitDegrees=45
  twistMotion=Free
 }
 {
  bone1Name=dyn_skirt_side_r_base
  bone2Name=spine_01
  swing1LimitDegrees=45
  swing1Motion=Free
  swing2LimitDegrees=45
  swing2Motion=Free
  twistLimitDegrees=45
  twistMotion=Free
 }
 {bone1Name=dyn_skirt_side_r_03, bone2Name=dyn_skirt_side_r_02, swing1LimitDegrees=5, swing2LimitDegrees=15, twistLimitDegrees=1}
 {bone1Name=dyn_skirt_side_r_02, bone2Name=dyn_skirt_side_r_01, swing1LimitDegrees=5, swing2LimitDegrees=15, twistLimitDegrees=1}
 {bone1Name=dyn_skirt_side_r_01, bone2Name=dyn_skirt_side_r_base, swing1LimitDegrees=10, swing2LimitDegrees=10, twistLimitDegrees=5}
 {bone1Name=dyn_skirt_side_l_03, bone2Name=dyn_skirt_side_l_02, swing1LimitDegrees=5, swing2LimitDegrees=15, twistLimitDegrees=1}
 {bone1Name=dyn_skirt_side_l_01, bone2Name=dyn_skirt_side_l_base, swing1LimitDegrees=10, swing2LimitDegrees=10, twistLimitDegrees=5}
 {bone1Name=dyn_skirt_side_l_02, bone2Name=dyn_skirt_side_l_01, swing1LimitDegrees=5, swing2LimitDegrees=15, twistLimitDegrees=1}
 {bone1Name=dyn_skirt_bk_r_03, bone2Name=dyn_skirt_bk_r_02, swing1LimitDegrees=5, swing2LimitDegrees=5, twistLimitDegrees=1}
 {bone1Name=dyn_skirt_bk_r_01, bone2Name=dyn_skirt_bk_r_base, swing1LimitDegrees=1, swing2LimitDegrees=10, twistLimitDegrees=1}
 {bone1Name=dyn_skirt_bk_r_02, bone2Name=dyn_skirt_bk_r_01, swing1LimitDegrees=5, swing2LimitDegrees=5, twistLimitDegrees=1}
 {bone1Name=dyn_skirt_bk_l_03, bone2Name=dyn_skirt_bk_l_02, swing1LimitDegrees=5, swing2LimitDegrees=5, twistLimitDegrees=1}
 {bone1Name=dyn_skirt_bk_l_02, bone2Name=dyn_skirt_bk_l_01, swing1LimitDegrees=5, swing2LimitDegrees=5, twistLimitDegrees=1}
 {bone1Name=dyn_skirt_bk_l_01, bone2Name=dyn_skirt_bk_l_base, swing1LimitDegrees=1, swing2LimitDegrees=10, twistLimitDegrees=1}
 {bone1Name=dyn_skirt_bk_centr_03, bone2Name=dyn_skirt_bk_centr_02, swing1LimitDegrees=5, swing2LimitDegrees=5, twistLimitDegrees=1}
 {bone1Name=dyn_skirt_bk_centr_02, bone2Name=dyn_skirt_bk_centr_01, swing1LimitDegrees=5, swing2LimitDegrees=5, twistLimitDegrees=1}
 {bone1Name=dyn_skirt_bk_centr_01, bone2Name=dyn_skirt_bk_centr_base, swing1LimitDegrees=1, swing2LimitDegrees=10, twistLimitDegrees=1}
 {bone1Name=thigh_r, bone2Name=pelvis, swing1LimitDegrees=45, swing2LimitDegrees=45, twistLimitDegrees=45}
 {bone1Name=dyn_obi_belt_down, bone2Name=dyn_obi_belt_down_base, swing1LimitDegrees=5, swing2LimitDegrees=30, twistLimitDegrees=5}
 {
  bone1Name=dyn_obi_belt_down_base
  bone2Name=dyn_skirt_bk_centr_01
  swing1LimitDegrees=45
  swing1Motion=Free
  swing2LimitDegrees=45
  swing2Motion=Free
  twistLimitDegrees=45
  twistMotion=Free
 }
]
