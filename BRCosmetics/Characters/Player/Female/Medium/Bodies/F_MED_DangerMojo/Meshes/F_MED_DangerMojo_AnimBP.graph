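# Post-process style cosmetic AnimBP for the DangerMojo female body, copying the attached parent's pose and adding secondary motion to clothing.
# Weighted constraints drive skirt bones from thigh helpers, then a Control Rig runs before jacket, obi and hood-strap constraints.
# A pelvis-space skirt RigidBody, bone offsets, back-jacket Look Ats and a post Control Rig finish the pose.
# The jacket RigidBody is unconnected; event-computed Gravity and LOD_1 go unread.
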AnimBlueprint /BRCosmetics/Characters/Player/Female/Medium/Bodies/F_MED_DangerMojo/Meshes/F_MED_DangerMojo_AnimBP
target_skeleton=&/F_MED_DangerMojo_Skeleton
graphs [
 AnimGraph {nodes [
   "Constraint - dyn_skirt_fr2_01_l_base (PP)"!(ComponentPose, Alpha=1, ConstraintWeights_0=0.2, ConstraintWeights_1:float=0.8) -> (Pose) {
    alphaBoolBlend {blendOption=Linear}
    boneToModify {boneName=dyn_skirt_fr2_01_l_base}
    constraintSetup [{targetBone {boneName=dyn_skirt_fr_01_l_base}}, {targetBone {boneName=dyn_skirt_side_l_base}}]
    constraintWeights [1,1]
   }
   RigidBody!(ComponentPose, Alpha, OverrideWorldGravity:Vector="0,0,-980", ExternalForce) -> (Pose) {
    bClampLinearTranslationLimitToRefPose=true
    bFreezeIncomingPoseOnStart=true
    bOverrideWorldGravity=true
    bTransferBoneVelocities=true
    overridePhysicsAsset=&/F_MED_DangerMojo_Physics_jaket
    simulationSpace=BaseBoneSpace
    alphaBoolBlend {blendOption=Linear}
    baseBoneRef {boneName=spine_04}
    componentAppliedLinearAccClamp=<3000 3000 3000>
    componentLinearAccScale=<0.4 0.4 0.4>
    componentLinearVelScale=<0.4 0.4 0.4>
    overrideWorldGravity {z=-980}
    simSpaceSettings {maxAngularAcceleration=3000, maxAngularVelocity=3000, maxLinearAcceleration=3000, maxLinearVelocity=3000, worldAlpha=0.5}
   }
   n1 = "Constraint - dyn_thigh_01_l (RR)"!(
    ComponentPose=Local To Component(LocalPose=Copy Pose From Mesh() {bUseAttachedParent=true})
    Alpha=1
    ConstraintWeights_0=1
    ConstraintWeights_1:float=0
   ) -> (Pose) {
    alphaBoolBlend {blendOption=Linear}
    boneToModify {boneName=dyn_thigh_01_l}
    constraintSetup [{transformType=Rotation, targetBone {boneName=thigh_l}}, {transformType=Rotation, targetBone {boneName=pelvis}}]
    constraintWeights [1,1]
   }
   n2 = "Constraint - dyn_thigh_01_r (RR)"!(ComponentPose=@n1, Alpha=1, ConstraintWeights_0=1, ConstraintWeights_1:float=0) -> (Pose) {
    alphaBoolBlend {blendOption=Linear}
    boneToModify {boneName=dyn_thigh_01_r}
    constraintSetup [{transformType=Rotation, targetBone {boneName=thigh_r}}, {transformType=Rotation, targetBone {boneName=pelvis}}]
    constraintWeights [1,1]
   }
   n3 = "Constraint - dyn_skirt_fr_01_base (RR)"!(ComponentPose=@n2, Alpha=1, ConstraintWeights_0=0.5, ConstraintWeights_1:float=0.5) -> (Pose) {
    alphaBoolBlend {blendOption=Linear}
    boneToModify {boneName=dyn_skirt_fr_01_base}
    constraintSetup [
     {transformType=Rotation, targetBone {boneName=dyn_skirt_fr_01_l_base}}
     {transformType=Rotation, targetBone {boneName=dyn_skirt_fr_01_r_base}}
    ]
    constraintWeights [1,1]
   }
   n4 = "Constraint - dyn_skirt_fr_01 (RR)"!(ComponentPose=@n3, Alpha=1, ConstraintWeights_0=0.5, ConstraintWeights_1:float=0.5) -> (Pose) {
    alphaBoolBlend {blendOption=Linear}
    boneToModify {boneName=dyn_skirt_fr_01}
    constraintSetup [{transformType=Rotation, targetBone {boneName=dyn_skirt_fr_01_l}}, {transformType=Rotation, targetBone {boneName=dyn_skirt_fr_01_r}}]
    constraintWeights [1,1]
   }
   n5 = "Constraint - dyn_skirt_fr_02 (RR)"!(ComponentPose=@n4, Alpha=1, ConstraintWeights_0=0.5, ConstraintWeights_1:float=0.5) -> (Pose) {
    alphaBoolBlend {blendOption=Linear}
    boneToModify {boneName=dyn_skirt_fr_02}
    constraintSetup [{transformType=Rotation, targetBone {boneName=dyn_skirt_fr_02_l}}, {transformType=Rotation, targetBone {boneName=dyn_skirt_fr_02_r}}]
    constraintWeights [1,1]
   }
   n6 = "Constraint - dyn_skirt_fr_03 (RR)"!(ComponentPose=@n5, Alpha=1, ConstraintWeights_0=0.5, ConstraintWeights_1:float=0.5) -> (Pose) {
    alphaBoolBlend {blendOption=Linear}
    boneToModify {boneName=dyn_skirt_fr_03}
    constraintSetup [{transformType=Rotation, targetBone {boneName=dyn_skirt_fr_03_l}}, {transformType=Rotation, targetBone {boneName=dyn_skirt_fr_03_r}}]
    constraintWeights [1,1]
   }
   n7 = "Constraint - dyn_skirt_fr_01_l_base (PP)"!(ComponentPose=@n6, Alpha=0.95, ConstraintWeights_0=0, ConstraintWeights_1:float=1) -> (Pose) {
    alphaBoolBlend {blendOption=Linear}
    boneToModify {boneName=dyn_skirt_fr_01_l_base}
    constraintSetup [{targetBone {boneName=pelvis}}, {targetBone {boneName=dyn_thigh_01_l}}]
    constraintWeights [1,1]
   }
   n8 = "Constraint - dyn_skirt_fr_01_l (PP)"!(ComponentPose=@n7, Alpha=0.95, ConstraintWeights_0=0, ConstraintWeights_1:float=1) -> (Pose) {
    alphaBoolBlend {blendOption=Linear}
    boneToModify {boneName=dyn_skirt_fr_01_l}
    constraintSetup [{targetBone {boneName=pelvis}}, {targetBone {boneName=dyn_thigh_01_l}}]
    constraintWeights [1,1]
   }
   n9 = "Constraint - dyn_skirt_fr_02_l (PP)"!(ComponentPose=@n8, Alpha=0.85, ConstraintWeights_0=0, ConstraintWeights_1:float=1) -> (Pose) {
    alphaBoolBlend {blendOption=Linear}
    boneToModify {boneName=dyn_skirt_fr_02_l}
    constraintSetup [{targetBone {boneName=pelvis}}, {targetBone {boneName=dyn_thigh_01_l}}]
    constraintWeights [1,1]
   }
   n10 = "Constraint - dyn_skirt_fr_03_l (PP)"!(ComponentPose=@n9, Alpha=0.6, ConstraintWeights_0=0, ConstraintWeights_1:float=1) -> (Pose) {
    alphaBoolBlend {blendOption=Linear}
    boneToModify {boneName=dyn_skirt_fr_03_l}
    constraintSetup [{targetBone {boneName=pelvis}}, {targetBone {boneName=dyn_thigh_01_l}}]
    constraintWeights [1,1]
   }
   n11 = "Constraint - dyn_skirt_fr2_01_l_base (PP)"(ComponentPose=@n10, Alpha=0.9, ConstraintWeights_0=0, ConstraintWeights_1=1) {
    alphaBoolBlend {blendOption=Linear}
    boneToModify {boneName=dyn_skirt_fr2_01_l_base}
    constraintSetup [{targetBone {boneName=pelvis}}, {targetBone {boneName=dyn_thigh_01_l}}]
    constraintWeights [1,1]
   }
   n12 = "Constraint - dyn_skirt_fr2_01_l (PP)"!(ComponentPose=@n11, Alpha=0.9, ConstraintWeights_0=0, ConstraintWeights_1:float=1) -> (Pose) {
    alphaBoolBlend {blendOption=Linear}
    boneToModify {boneName=dyn_skirt_fr2_01_l}
    constraintSetup [{targetBone {boneName=pelvis}}, {targetBone {boneName=dyn_thigh_01_l}}]
    constraintWeights [1,1]
   }
   n13 = "Constraint - dyn_skirt_fr2_02_l (PP)"!(ComponentPose=@n12, Alpha=0.8, ConstraintWeights_0=0, ConstraintWeights_1:float=1) -> (Pose) {
    alphaBoolBlend {blendOption=Linear}
    boneToModify {boneName=dyn_skirt_fr2_02_l}
    constraintSetup [{targetBone {boneName=pelvis}}, {targetBone {boneName=dyn_thigh_01_l}}]
    constraintWeights [1,1]
   }
   n14 = "Constraint - dyn_skirt_fr2_03_l (PP)"!(ComponentPose=@n13, Alpha=0.6, ConstraintWeights_0=0, ConstraintWeights_1:float=1) -> (Pose) {
    alphaBoolBlend {blendOption=Linear}
    boneToModify {boneName=dyn_skirt_fr2_03_l}
    constraintSetup [{targetBone {boneName=pelvis}}, {targetBone {boneName=dyn_thigh_01_l}}]
    constraintWeights [1,1]
   }
   n15 = "Constraint - dyn_skirt_fr_01_r_base (PP)"!(ComponentPose=@n14, Alpha=0.9, ConstraintWeights_0=0, ConstraintWeights_1:float=1) -> (Pose) {
    alphaBoolBlend {blendOption=Linear}
    boneToModify {boneName=dyn_skirt_fr_01_r_base}
    constraintSetup [{targetBone {boneName=pelvis}}, {targetBone {boneName=dyn_thigh_01_r}}]
    constraintWeights [1,1]
   }
   n16 = "Constraint - dyn_skirt_fr_01_r (PP)"!(ComponentPose=@n15, Alpha=0.9, ConstraintWeights_0=0, ConstraintWeights_1:float=1) -> (Pose) {
    alphaBoolBlend {blendOption=Linear}
    boneToModify {boneName=dyn_skirt_fr_01_r}
    constraintSetup [{targetBone {boneName=pelvis}}, {targetBone {boneName=dyn_thigh_01_r}}]
    constraintWeights [1,1]
   }
   n17 = "Constraint - dyn_skirt_fr_02_r (PP)"!(ComponentPose=@n16, Alpha=0.5, ConstraintWeights_0=0, ConstraintWeights_1:float=1) -> (Pose) {
    alphaBoolBlend {blendOption=Linear}
    boneToModify {boneName=dyn_skirt_fr_02_r}
    constraintSetup [{targetBone {boneName=pelvis}}, {targetBone {boneName=dyn_thigh_01_r}}]
    constraintWeights [1,1]
   }
   n18 = "Constraint - dyn_skirt_fr_03_r (PP)"!(ComponentPose=@n17, Alpha=0.5, ConstraintWeights_0=0, ConstraintWeights_1:float=1) -> (Pose) {
    alphaBoolBlend {blendOption=Linear}
    boneToModify {boneName=dyn_skirt_fr_03_r}
    constraintSetup [{targetBone {boneName=pelvis}}, {targetBone {boneName=dyn_thigh_01_r}}]
    constraintWeights [1,1]
   }
   n19 = "Constraint - dyn_skirt_fr2_01_r_base (PP)"!(ComponentPose=@n18, Alpha=0.9, ConstraintWeights_0=0, ConstraintWeights_1:float=1) -> (Pose) {
    alphaBoolBlend {blendOption=Linear}
    boneToModify {boneName=dyn_skirt_fr2_01_r_base}
    constraintSetup [{targetBone {boneName=pelvis}}, {targetBone {boneName=dyn_thigh_01_r}}]
    constraintWeights [1,1]
   }
   n20 = "Constraint - dyn_skirt_fr2_01_r (PP)"!(ComponentPose=@n19, Alpha=0.9, ConstraintWeights_0=0, ConstraintWeights_1:float=1) -> (Pose) {
    alphaBoolBlend {blendOption=Linear}
    boneToModify {boneName=dyn_skirt_fr2_01_r}
    constraintSetup [{targetBone {boneName=pelvis}}, {targetBone {boneName=dyn_thigh_01_r}}]
    constraintWeights [1,1]
   }
   n21 = "Constraint - dyn_skirt_fr2_02_r (PP)"!(ComponentPose=@n20, Alpha=0.5, ConstraintWeights_0=0, ConstraintWeights_1:float=1) -> (Pose) {
    alphaBoolBlend {blendOption=Linear}
    boneToModify {boneName=dyn_skirt_fr2_02_r}
    constraintSetup [{targetBone {boneName=pelvis}}, {targetBone {boneName=dyn_thigh_01_r}}]
    constraintWeights [1,1]
   }
   n22 = "Constraint - dyn_skirt_fr2_03_r (PP)"!(ComponentPose=@n21, Alpha=0.5, ConstraintWeights_0=0, ConstraintWeights_1:float=1) -> (Pose) {
    alphaBoolBlend {blendOption=Linear}
    boneToModify {boneName=dyn_skirt_fr2_03_r}
    constraintSetup [{targetBone {boneName=pelvis}}, {targetBone {boneName=dyn_thigh_01_r}}]
    constraintWeights [1,1]
   }
   n23 = "Constraint - dyn_skirt_fr_01_base (PP)"!(ComponentPose=@n22, Alpha=1, ConstraintWeights_0=0.5, ConstraintWeights_1:float=0.5) -> (Pose) {
    alphaBoolBlend {blendOption=Linear}
    boneToModify {boneName=dyn_skirt_fr_01_base}
    constraintSetup [{targetBone {boneName=dyn_skirt_fr_01_l_base}}, {targetBone {boneName=dyn_skirt_fr_01_r_base}}]
    constraintWeights [1,1]
   }
   n24 = "Constraint - dyn_skirt_fr_01 (PP)"!(ComponentPose=@n23, Alpha=1, ConstraintWeights_0=0.5, ConstraintWeights_1:float=0.5) -> (Pose) {
    alphaBoolBlend {blendOption=Linear}
    boneToModify {boneName=dyn_skirt_fr_01}
    constraintSetup [{targetBone {boneName=dyn_skirt_fr_01_l}}, {targetBone {boneName=dyn_skirt_fr_01_r}}]
    constraintWeights [1,1]
   }
   n25 = "Constraint - dyn_skirt_fr_02 (PP)"!(ComponentPose=@n24, Alpha=1, ConstraintWeights_0=0.5, ConstraintWeights_1:float=0.5) -> (Pose) {
    alphaBoolBlend {blendOption=Linear}
    boneToModify {boneName=dyn_skirt_fr_02}
    constraintSetup [{targetBone {boneName=dyn_skirt_fr_02_l}}, {targetBone {boneName=dyn_skirt_fr_02_r}}]
    constraintWeights [1,1]
   }
   n26 = "Constraint - dyn_skirt_fr_03 (PP)"!(ComponentPose=@n25, Alpha=1, ConstraintWeights_0=0.5, ConstraintWeights_1:float=0.5) -> (Pose) {
    alphaBoolBlend {blendOption=Linear}
    boneToModify {boneName=dyn_skirt_fr_03}
    constraintSetup [{targetBone {boneName=dyn_skirt_fr_03_l}}, {targetBone {boneName=dyn_skirt_fr_03_r}}]
    constraintWeights [1,1]
   }
   n27 = "Constraint - dyn_skirt_side_r_base (PP)"!(ComponentPose=@n26, Alpha=1, ConstraintWeights_0=0.7, ConstraintWeights_1:float=0.3) -> (Pose) {
    alphaBoolBlend {blendOption=Linear}
    boneToModify {boneName=dyn_skirt_side_r_base}
    constraintSetup [{targetBone {boneName=dyn_skirt_fr_01_r_base}}, {targetBone {boneName=dyn_skirt_bk_r_base}}]
    constraintWeights [1,1]
   }
   n28 = "Constraint - dyn_skirt_fr2_01_r_base (PP)"(ComponentPose=@n27, Alpha=1, ConstraintWeights_0=0.7, ConstraintWeights_1=0.3) {
    alphaBoolBlend {blendOption=Linear}
    boneToModify {boneName=dyn_skirt_fr2_01_r_base}
    constraintSetup [{targetBone {boneName=dyn_skirt_fr_01_r_base}}, {targetBone {boneName=dyn_skirt_side_r_base}}]
    constraintWeights [1,1]
   }
   n29 = "Constraint - dyn_skirt_side_l_base (PP)"!(ComponentPose=@n28, Alpha=1, ConstraintWeights_0=0.6, ConstraintWeights_1:float=0.4) -> (Pose) {
    alphaBoolBlend {blendOption=Linear}
    boneToModify {boneName=dyn_skirt_side_l_base}
    constraintSetup [{targetBone {boneName=dyn_skirt_fr_01_l_base}}, {targetBone {boneName=dyn_skirt_bk_l_base}}]
    constraintWeights [1,1]
   }
   n30 = Control Rig(Source=Component To Local(ComponentPose=@n29)) {
    alphaBoolBlend {blendOption=Linear}
    controlRigAssetReference {blueprintRigClass=&/F_MED_DangerMojo_CtrlRig.F_MED_DangerMojo_CtrlRig_C}
   }
   n31 = "Constraint - dyn_jaket_bk_centr_base (PPPPPPPP)"!(
    ComponentPose=Local To Component(LocalPose=@n30)
    Alpha=1
    ConstraintWeights_0=0.3474
    ConstraintWeights_1:float=0.1686
    ConstraintWeights_2:float=0.1598
    ConstraintWeights_3:float=0.1334
    ConstraintWeights_4:float=0.0664
    ConstraintWeights_5:float=0.0631
    ConstraintWeights_6:float=0.0508
    ConstraintWeights_7:float=0.0105
   ) -> (Pose) {
    alphaBoolBlend {blendOption=Linear}
    boneToModify {boneName=dyn_jaket_bk_centr_base}
    constraintSetup [
     {targetBone {boneName=spine_04}}
     {targetBone {boneName=spine_02}}
     {targetBone {boneName=spine_05}}
     {targetBone {boneName=spine_03}}
     {targetBone {boneName=clavicle_l}}
     {targetBone {boneName=clavicle_r}}
     {targetBone {boneName=spine_01}}
     {targetBone {boneName=pelvis}}
    ]
    constraintWeights [1,1,1,1,1,1,1,1]
   }
   n32 = "Constraint - dyn_jaket_bk_side_r_base (PPPPPPP)"!(
    ComponentPose=@n31
    Alpha=1
    ConstraintWeights_0=0.4419
    ConstraintWeights_1:float=0.1584
    ConstraintWeights_2:float=0.1281
    ConstraintWeights_3:float=0.091
    ConstraintWeights_4:float=0.0825
    ConstraintWeights_5:float=0.0767
    ConstraintWeights_6:float=0.0215
   ) -> (Pose) {
    alphaBoolBlend {blendOption=Linear}
    boneToModify {boneName=dyn_jaket_bk_side_r_base}
    constraintSetup [
     {targetBone {boneName=spine_04}}
     {targetBone {boneName=clavicle_r}}
     {targetBone {boneName=spine_05}}
     {targetBone {boneName=spine_02}}
     {targetBone {boneName=upperarm_twist_01_r}}
     {targetBone {boneName=spine_03}}
     {targetBone {boneName=spine_01}}
    ]
    constraintWeights [1,1,1,1,1,1,1]
   }
   n33 = "Constraint - dyn_jaket_fr_r_base (PPPPP)"!(
    ComponentPose=@n32
    Alpha=1
    ConstraintWeights_0=0.6905
    ConstraintWeights_1:float=0.2072
    ConstraintWeights_2:float=0.0519
    ConstraintWeights_3:float=0.0486
    ConstraintWeights_4:float=0.0018
   ) -> (Pose) {
    alphaBoolBlend {blendOption=Linear}
    boneToModify {boneName=dyn_jaket_fr_r_base}
    constraintSetup [
     {targetBone {boneName=spine_04}}
     {targetBone {boneName=deform_pec_r}}
     {targetBone {boneName=spine_02}}
     {targetBone {boneName=spine_03}}
     {targetBone {boneName=spine_05}}
    ]
    constraintWeights [1,1,1,1,1]
   }
   n34 = "Constraint - dyn_jaket_fr_side_r_base (PPPPPPPPP)"!(
    ComponentPose=@n33
    Alpha=1
    ConstraintWeights_0=0.6371
    ConstraintWeights_1:float=0.093
    ConstraintWeights_2:float=0.0701
    ConstraintWeights_3:float=0.0668
    ConstraintWeights_4:float=0.0505
    ConstraintWeights_5:float=0.044
    ConstraintWeights_6:float=0.0267
    ConstraintWeights_7:float=0.0108
    ConstraintWeights_8:float=0.0011
   ) -> (Pose) {
    alphaBoolBlend {blendOption=Linear}
    boneToModify {boneName=dyn_jaket_fr_side_r_base}
    constraintSetup [
     {targetBone {boneName=spine_04}}
     {targetBone {boneName=spine_02}}
     {targetBone {boneName=spine_05}}
     {targetBone {boneName=spine_03}}
     {targetBone {boneName=clavicle_r}}
     {targetBone {boneName=deform_pec_r}}
     {targetBone {boneName=upperarm_twist_01_r}}
     {targetBone {boneName=spine_01}}
     {targetBone {boneName=pelvis}}
    ]
    constraintWeights [1,1,1,1,1,1,1,1,1]
   }
   n35 = "Constraint - dyn_jaket_fr_l_base (PPPPPP)"!(
    ComponentPose=@n34
    Alpha=1
    ConstraintWeights_0=0.5836
    ConstraintWeights_1:float=0.3308
    ConstraintWeights_2:float=0.0286
    ConstraintWeights_3:float=0.0276
    ConstraintWeights_4:float=0.0241
    ConstraintWeights_5:float=0.0051
   ) -> (Pose) {
    alphaBoolBlend {blendOption=Linear}
    boneToModify {boneName=dyn_jaket_fr_l_base}
    constraintSetup [
     {targetBone {boneName=spine_04}}
     {targetBone {boneName=deform_pec_l}}
     {targetBone {boneName=spine_02}}
     {targetBone {boneName=spine_03}}
     {targetBone {boneName=spine_05}}
     {targetBone {boneName=clavicle_l}}
    ]
    constraintWeights [1,1,1,1,1,1]
   }
   n36 = "Constraint - dyn_jaket_fr_side_l_base (PPPPPPPPP)"!(
    ComponentPose=@n35
    Alpha=1
    ConstraintWeights_0=0.6363
    ConstraintWeights_1:float=0.0913
    ConstraintWeights_2:float=0.0713
    ConstraintWeights_3:float=0.066
    ConstraintWeights_4:float=0.0515
    ConstraintWeights_5:float=0.0448
    ConstraintWeights_6:float=0.0276
    ConstraintWeights_7:float=0.0102
    ConstraintWeights_8:float=0.001
   ) -> (Pose) {
    alphaBoolBlend {blendOption=Linear}
    boneToModify {boneName=dyn_jaket_fr_side_l_base}
    constraintSetup [
     {targetBone {boneName=spine_04}}
     {targetBone {boneName=spine_02}}
     {targetBone {boneName=spine_05}}
     {targetBone {boneName=spine_03}}
     {targetBone {boneName=clavicle_l}}
     {targetBone {boneName=deform_pec_l}}
     {targetBone {boneName=upperarm_twist_01_l}}
     {targetBone {boneName=spine_01}}
     {targetBone {boneName=pelvis}}
    ]
    constraintWeights [1,1,1,1,1,1,1,1,1]
   }
   n37 = "Constraint - dyn_jaket_bk_side_l_base (PPPPPPP)"!(
    ComponentPose=@n36
    Alpha=1
    ConstraintWeights_0=0.442
    ConstraintWeights_1:float=0.1584
    ConstraintWeights_2:float=0.1281
    ConstraintWeights_3:float=0.0909
    ConstraintWeights_4:float=0.0825
    ConstraintWeights_5:float=0.0767
    ConstraintWeights_6:float=0.0215
   ) -> (Pose) {
    alphaBoolBlend {blendOption=Linear}
    boneToModify {boneName=dyn_jaket_bk_side_l_base}
    constraintSetup [
     {targetBone {boneName=spine_04}}
     {targetBone {boneName=clavicle_l}}
     {targetBone {boneName=spine_05}}
     {targetBone {boneName=spine_02}}
     {targetBone {boneName=upperarm_twist_01_l}}
     {targetBone {boneName=spine_03}}
     {targetBone {boneName=spine_01}}
    ]
    constraintWeights [1,1,1,1,1,1,1]
   }
   n38 = "Constraint - dyn_obi_belt_up_base (PPPPPPP)"!(
    ComponentPose=@n37
    Alpha=1
    ConstraintWeights_0=0.4652
    ConstraintWeights_1:float=0.1688
    ConstraintWeights_2:float=0.1251
    ConstraintWeights_3:float=0.1211
    ConstraintWeights_4:float=0.1108
    ConstraintWeights_5:float=0.0083
    ConstraintWeights_6:float=0.0007
   ) -> (Pose) {
    alphaBoolBlend {blendOption=Linear}
    boneToModify {boneName=dyn_obi_belt_up_base}
    constraintSetup [
     {targetBone {boneName=spine_02}}
     {targetBone {boneName=pelvis}}
     {targetBone {boneName=spine_04}}
     {targetBone {boneName=spine_03}}
     {targetBone {boneName=spine_01}}
     {targetBone {boneName=spine_05}}
     {targetBone {boneName=clavicle_r}}
    ]
    constraintWeights [1,1,1,1,1,1,1]
   }
   n39 = "Constraint - dyn_hood_strap_r_base (PPPPPP)"!(
    ComponentPose=@n38
    Alpha=1
    ConstraintWeights_0=0.3699
    ConstraintWeights_1:float=0.292
    ConstraintWeights_2:float=0.1911
    ConstraintWeights_3:float=0.0952
    ConstraintWeights_4:float=0.0317
    ConstraintWeights_5:float=0.0202
   ) -> (Pose) {
    alphaBoolBlend {blendOption=Linear}
    boneToModify {boneName=dyn_hood_strap_r_base}
    constraintSetup [
     {targetBone {boneName=deform_pec_r}}
     {targetBone {boneName=spine_04}}
     {targetBone {boneName=spine_05}}
     {targetBone {boneName=clavicle_r}}
     {targetBone {boneName=upperarm_twist_01_r}}
     {targetBone {boneName=deform_upperarm_fr_r}}
    ]
    constraintWeights [1,1,1,1,1,1]
   }
   n40 = "Constraint - dyn_hood_strap_l_base (PPPPPP)"!(
    ComponentPose=@n39
    Alpha=1
    ConstraintWeights_0=0.3953
    ConstraintWeights_1:float=0.2997
    ConstraintWeights_2:float=0.1999
    ConstraintWeights_3:float=0.0752
    ConstraintWeights_4:float=0.02
    ConstraintWeights_5:float=0.01
   ) -> (Pose) {
    alphaBoolBlend {blendOption=Linear}
    boneToModify {boneName=dyn_hood_strap_l_base}
    constraintSetup [
     {targetBone {boneName=deform_pec_l}}
     {targetBone {boneName=spine_04}}
     {targetBone {boneName=spine_05}}
     {targetBone {boneName=clavicle_l}}
     {targetBone {boneName=upperarm_twist_01_l}}
     {targetBone {boneName=deform_upperarm_fr_l}}
    ]
    constraintWeights [1,1,1,1,1,1]
   }
   n41 = "Constraint - dyn_obi_belt_down_base (RRRRRRRRRRRRRR)"!(
    ComponentPose=@n40
    Alpha=1
    ConstraintWeights_0=0.0115
    ConstraintWeights_1:float=0.9192
    ConstraintWeights_2:float=0.0028
    ConstraintWeights_3:float=0.0071
    ConstraintWeights_4:float=0.0051
    ConstraintWeights_5:float=0.0119
    ConstraintWeights_6:float=0.0046
    ConstraintWeights_7:float=0.0058
    ConstraintWeights_8:float=0.0043
    ConstraintWeights_9:float=0.006
    ConstraintWeights_10:float=0.0078
    ConstraintWeights_11:float=0.0097
    ConstraintWeights_12:float=0.0042
    ConstraintWeights_13:float=1
   ) -> (Pose) {
    alphaBoolBlend {blendOption=Linear}
    boneToModify {boneName=dyn_obi_belt_down_base}
    constraintSetup [
     {transformType=Rotation, targetBone {boneName=spine_02}}
     {transformType=Rotation, targetBone {boneName=pelvis}}
     {transformType=Rotation, targetBone {boneName=dyn_skirt_fr2_02_r}}
     {transformType=Rotation, targetBone {boneName=dyn_skirt_bk_r_03}}
     {transformType=Rotation, targetBone {boneName=dyn_skirt_side_r_01}}
     {transformType=Rotation, targetBone {boneName=dyn_skirt_bk_centr_03}}
     {transformType=Rotation, targetBone {boneName=dyn_skirt_fr2_02_l}}
     {transformType=Rotation, targetBone {boneName=dyn_skirt_bk_l_01}}
     {transformType=Rotation, targetBone {boneName=dyn_skirt_bk_r_02}}
     {transformType=Rotation, targetBone {boneName=dyn_skirt_bk_l_02}}
     {transformType=Rotation, targetBone {boneName=dyn_skirt_side_l_01}}
     {transformType=Rotation, targetBone {boneName=dyn_skirt_bk_l_03}}
     {transformType=Rotation, targetBone {boneName=dyn_skirt_bk_r_01}}
     {transformType=Rotation, targetBone {boneName=None}}
    ]
    constraintWeights [1,1,1,1,1,1,1,1,1,1,1,1,1,1]
   }
   n42 = RigidBody(ComponentPose=@n41, OverrideWorldGravity="0,0,-980") {
    bClampLinearTranslationLimitToRefPose=true
    bFreezeIncomingPoseOnStart=true
    bOverrideWorldGravity=true
    bTransferBoneVelocities=true
    overridePhysicsAsset=&/F_MED_DangerMojo_Physics_skirt
    simulationSpace=BaseBoneSpace
    alphaBoolBlend {blendOption=Linear}
    baseBoneRef {boneName=pelvis}
    componentAppliedLinearAccClamp=<3000 3000 3000>
    componentLinearAccScale=<0.6 0.6 0.6>
    componentLinearVelScale=<0.4 0.4 0.4>
    overrideWorldGravity {z=-980}
    simSpaceSettings {maxAngularAcceleration=3000, maxAngularVelocity=3000, maxLinearAcceleration=3000, maxLinearVelocity=3000, worldAlpha=0.5}
   }
   n43 = "Transform (Modify) Bone - Bone: extra_legs"(ComponentPose=@n42, Alpha=1, Translation="-1,-2,0", Rotation="0,0,0", Scale="1,1,1") {
    translationMode=BMM_Additive
    translationSpace=BCS_BoneSpace
    alphaBoolBlend {blendOption=Linear}
    boneToModify {boneName=extra_legs}
    translation=<-1 -2>
   }
   n44 = "Constraint - dyn_jaket_bk_centr (PPPPPPPP)"!(
    ComponentPose=@n43
    Alpha=1
    ConstraintWeights_0=0.3474
    ConstraintWeights_1:float=0.1686
    ConstraintWeights_2:float=0.1598
    ConstraintWeights_3:float=0.1334
    ConstraintWeights_4:float=0.0664
    ConstraintWeights_5:float=0.0631
    ConstraintWeights_6:float=0.0508
    ConstraintWeights_7:float=0.0105
   ) -> (Pose) {
    alphaBoolBlend {blendOption=Linear}
    boneToModify {boneName=dyn_jaket_bk_centr}
    constraintSetup [
     {targetBone {boneName=spine_04}}
     {targetBone {boneName=spine_02}}
     {targetBone {boneName=spine_05}}
     {targetBone {boneName=spine_03}}
     {targetBone {boneName=clavicle_l}}
     {targetBone {boneName=clavicle_r}}
     {targetBone {boneName=spine_01}}
     {targetBone {boneName=pelvis}}
    ]
    constraintWeights [1,1,1,1,1,1,1,1]
   }
   n45 = "Transform (Modify) Bone - Bone: extra_choker"!(ComponentPose=@n44, bAlphaBoolEnabled:bool=True, Translation="0,0,0", Rotation="0,0,0", Scale="1.4,1,1.4") -> (Pose) {
    alphaInputType=Bool
    scaleMode=BMM_Additive
    scaleSpace=BCS_BoneSpace
    alphaBoolBlend {blendOption=Linear}
    boneToModify {boneName=extra_choker}
    scale {x=1.4, z=1.4}
   }
   n46 = Look At - Bone: dyn_jaket_bk_side_l_base(ComponentPose=@n45, Alpha=1) {
    alphaBoolBlend {blendOption=Linear}
    boneToModify {boneName=dyn_jaket_bk_side_l_base}
    lookAtLocation=<30 6>
    lookAtTarget {boneReference {boneName=dyn_skirt_bk_l_base}}
    lookAt_Axis {axis=<1 0>}
    lookUp_Axis {axis=<0 1>}
   }
   n47 = Look At - Bone: dyn_jaket_bk_centr_base(ComponentPose=@n46, Alpha=1) {
    alphaBoolBlend {blendOption=Linear}
    boneToModify {boneName=dyn_jaket_bk_centr_base}
    lookAtLocation {x=30, z=10}
    lookAtTarget {boneReference {boneName=dyn_obi_belt_up_base}}
    lookAt_Axis {axis=<1 0>}
    lookUp_Axis {axis=<0 1>}
   }
   n48 = Look At - Bone: dyn_jaket_bk_side_r_base(ComponentPose=@n47, Alpha=1) {
    alphaBoolBlend {blendOption=Linear}
    boneToModify {boneName=dyn_jaket_bk_side_r_base}
    lookAtLocation=<30 -6>
    lookAtTarget {boneReference {boneName=dyn_skirt_bk_r_base}}
    lookAt_Axis {axis=<1 0>}
    lookUp_Axis {axis=<0 1>}
   }
   n49 = Control Rig(Source=Component To Local(ComponentPose=@n48)) {
    alphaBoolBlend {blendOption=Linear}
    controlRigAssetReference {blueprintRigClass=&/F_MED_DangerMojo_CtrlRig_post.F_MED_DangerMojo_CtrlRig_post_C}
   }
   Output Pose(Result=@n49) {layerGroup=None}
  ]}
 EventGraph {nodes [
   Event BlueprintUpdateAnimation()
   > n1 = Sequence()
   > n2 = Gravity Override State Function DEPRECATED(
    CharacterPartAnimInstance=Self-Reference()
    GameTimeInSeconds=GetGameTimeInSeconds()
    GravityOverrideParams=Get Gravity Override Params()
    RBAN Params=Get RBAN Params()
   )
   > Set Gravity(Gravity:Vector=@n2) -> (Output_Get:Vector)
   Set LOD_1(execute=@n1.then_1, LOD_1:bool="Equal (Integer)"(A=Get CurrentLOD(), B=1)) -> (Output_Get:bool)
  ]}
]
variables [
 Gravity {type=Vector}
 Gravity Override Params {type=GravityOverrideParamsStruct}
 LOD_1 {type=boolean}
 RBAN Params {
  type=RBANParamsStruct
  default {
   dBNO {simSpaceSettings {
     maxAngularAcceleration=1000
     maxAngularVelocity=3
     maxLinearAcceleration=300
     maxLinearVelocity=700
     velocityScaleZ=0
     externalLinearDragV=<0 0.5 0>
    }}
   dBNO_Carried {simSpaceSettings {
     maxAngularAcceleration=1000
     maxAngularVelocity=3
     maxLinearAcceleration=300
     maxLinearVelocity=700
     velocityScaleZ=0
     externalLinearDragV=<0 0.5 0>
    }}
   emoteOrMelee {simSpaceSettings {
     maxAngularAcceleration=1000
     maxAngularVelocity=3
     maxLinearAcceleration=300
     maxLinearVelocity=700
     velocityScaleZ=0
     externalLinearDragV=<0 0.5 0>
    }}
   front_End {simSpaceSettings {
     maxAngularAcceleration=1000
     maxAngularVelocity=3
     maxLinearAcceleration=300
     maxLinearVelocity=700
     velocityScaleZ=0
     externalLinearDragV=<0 0.5 0>
    }}
   onGround_Crouch_Moving {simSpaceSettings {
     maxAngularAcceleration=1000
     maxAngularVelocity=3
     maxLinearAcceleration=300
     maxLinearVelocity=700
     velocityScaleZ=0
     externalLinearDragV=<0 0.5 0>
    }}
   onGround_Falling {simSpaceSettings {
     maxAngularAcceleration=1000
     maxAngularVelocity=3
     maxLinearAcceleration=300
     maxLinearVelocity=700
     velocityScaleZ=0
     externalLinearDragV=<0 0.5 0>
    }}
   onGround_Moving {simSpaceSettings {
     maxAngularAcceleration=1000
     maxAngularVelocity=3
     maxLinearAcceleration=300
     maxLinearVelocity=700
     velocityScaleZ=0
     externalLinearDragV=<0 0.5 0>
    }}
   onGround_Standing {simSpaceSettings {
     maxAngularAcceleration=1000
     maxAngularVelocity=3
     maxLinearAcceleration=300
     maxLinearVelocity=700
     velocityScaleZ=0
     externalLinearDragV=<0 0.5 0>
    }}
   rail_Grinding {simSpaceSettings {
     maxAngularAcceleration=1000
     maxAngularVelocity=3
     maxLinearAcceleration=300
     maxLinearVelocity=700
     velocityScaleZ=0
     worldAlpha=0.1
     externalLinearDragV=<0 0.5 0>
    }}
   riding_Moving {simSpaceSettings {
     maxAngularAcceleration=1000
     maxAngularVelocity=3
     maxLinearAcceleration=300
     maxLinearVelocity=700
     velocityScaleZ=0
     externalLinearDragV=<0 0.5 0>
    }}
   riding_Moving_Targeting {simSpaceSettings {
     maxAngularAcceleration=1000
     maxAngularVelocity=3
     maxLinearAcceleration=300
     maxLinearVelocity=700
     velocityScaleZ=0
     externalLinearDragV=<0 0.5 0>
    }}
   riding_Standing {simSpaceSettings {
     maxAngularAcceleration=1000
     maxAngularVelocity=3
     maxLinearAcceleration=300
     maxLinearVelocity=700
     velocityScaleZ=0
     externalLinearDragV=<0 0.5 0>
    }}
   skydive_DivingDown {simSpaceSettings {
     maxAngularAcceleration=1000
     maxAngularVelocity=3
     maxLinearAcceleration=300
     maxLinearVelocity=700
     velocityScaleZ=0
     externalLinearDragV=<0 0.5 0>
    }}
   skydive_FloatingDown {
    componentLinearAccScale=<0 0.5 0>
    componentLinearVelScale=<0 0.5 0>
    simSpaceSettings {
     maxAngularAcceleration=1000
     maxAngularVelocity=3
     maxLinearAcceleration=300
     maxLinearVelocity=700
     velocityScaleZ=0
     externalLinearDragV=<0 0.5 0>
    }
   }
   skydive_GoingUp {simSpaceSettings {
     maxAngularAcceleration=1000
     maxAngularVelocity=3
     maxLinearAcceleration=300
     maxLinearVelocity=700
     velocityScaleZ=0
     externalLinearDragV=<0 0.5 0>
    }}
   skydive_Parachuting {
    componentLinearAccScale=<0 0.1 0.3>
    componentLinearVelScale=<0 0.1 0.3>
    simSpaceSettings {
     maxAngularAcceleration=1000
     maxAngularVelocity=3
     maxLinearAcceleration=300
     maxLinearVelocity=700
     velocityScaleZ=0
     externalLinearDragV=<0 0.5 0>
    }
   }
   skydive_Surfing {simSpaceSettings {
     maxAngularAcceleration=1000
     maxAngularVelocity=3
     maxLinearAcceleration=300
     maxLinearVelocity=700
     velocityScaleZ=0
     externalLinearDragV=<0 0.5 0>
    }}
   swimming {simSpaceSettings {
     maxAngularAcceleration=1000
     maxAngularVelocity=3
     maxLinearAcceleration=300
     maxLinearVelocity=700
     velocityScaleZ=0
     externalLinearDragV=<0 0.5 0>
    }}
   tactical_Sprint {simSpaceSettings {
     maxAngularAcceleration=1000
     maxAngularVelocity=3
     maxLinearAcceleration=300
     maxLinearVelocity=700
     velocityScaleZ=0
     externalLinearDragV=<0 0.5 0>
    }}
   ziplining {simSpaceSettings {
     maxAngularAcceleration=1000
     maxAngularVelocity=3
     maxLinearAcceleration=300
     maxLinearVelocity=700
     velocityScaleZ=0
     externalLinearDragV=<0 0.5 0>
    }}
  }
 }
]
