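# Post-process Control Rig for the F_MED_DangerMojo face accessory that corrects dynamic hair bones and a right upper-arm helper bone.
# Smoothed aim input partly constrains left hair bones to spine_05 and offsets one base.
# Spherical pose readers on head, right clavicle and upper arm weight corrective offsets, while a head-to-hair distance check pushes right hair.
# Another head reader, damped by moving_2D and aim or overridden by FE_idel, blends a broader hair pose.
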ControlRigBlueprint /BRCosmetics/Characters/Player/Female/Medium/Bodies/F_MED_DangerMojo/Meshes/Parts/F_MED_DangerMojo_FaceAcc_post_CtrlRig
graphs [
 Multiply_2_2_ContainedGraph {nodes [
   n1 = RigVMFunctionEntryNode()
   RigVMFunctionReturnNode(Result=&RigVMFunction_MathDoubleMul(B=@n1.C, A=&RigVMFunction_MathDoubleMul(A=@n1.A, B=@n1.B).Result).Result)
  ]}
 RigVMFunctionLibrary {links [], nodes []}
 RigVMFunction_Sequence_3_1_ContainedGraph {nodes [
   n1 = RigVMFunctionEntryNode()
   n2 = &RigVMFunction_Sequence(ExecuteContext=@n1.ExecuteContext)
   n3 = &RigVMFunction_Sequence(ExecuteContext=@n2.B)
   RigVMFunctionReturnNode(A=@n2.A, B=@n3.A, C=@n3.B)
  ]}
 RigVMModel {nodes [
   RigUnit_BeginExecution = &RigUnit_BeginExecution()
   n1 = &RigVMFunction_Sequence(ExecuteContext=@RigUnit_BeginExecution.ExecutePin)
   AlphaInterp = &RigVMFunction_AlphaInterp(
    InterpSpeedDecreasing=3
    InterpSpeedIncreasing=3
    Scale=1
    bClampResult=False
    Value=&RigVMFunction_MathBoolToFloat(Value=RigVMVariableNode() {variable=aim}.Value).Result
   )
   Multiply = &RigVMFunction_MathDoubleMul(B=0.5, A=@AlphaInterp.Result)
   &RigUnit_ParentConstraint(
    AdvancedSettings=(InterpolationType=Average,RotationOrderForFilter=XZY)
    Child=(Type=Bone,Name="dyn_hair_side_bk_03_l")
    Parents=((Item=(Type=Bone,Name="spine_05"),Weight=1))
    Weight=@Multiply.Result
    ExecutePin=@n1.A
   )
   > &RigUnit_ModifyTransforms(
    ItemToModify=((Item=(Type=Bone,Name="dyn_hair_side_bk_l_base"),Transform=(Rotation=(X=0,Y=0,Z=0,W=1),Translation=(X=1.237,Y=0.292,Z=-1.017),Scale3D=(X=1,Y=1,Z=1))))
    WeightMinimum=0
    Weight=@AlphaInterp.Result
   )
   > &RigUnit_ParentConstraint(
    AdvancedSettings=(InterpolationType=Average,RotationOrderForFilter=XZY)
    Child=(Type=Bone,Name="dyn_hair_side_l_base")
    Parents=((Item=(Type=Bone,Name="spine_05"),Weight=1))
    Weight=&RigVMFunction_MathDoubleMul(B=0.3, A=@AlphaInterp.Result).Result
   )
   n2 = &RigVMFunction_Sequence(ExecuteContext=@n1.B)
   RigVMFunction_Sequence_3 = RigVMAggregateNode(ExecuteContext=@n2.B)
   n3 = &RigUnit_SphericalPoseReader(
    ActiveRegionScaleFactors=(PositiveWidth=1,NegativeWidth=0.5,PositiveHeight=0.75,NegativeHeight=0.75)
    ActiveRegionSize=0.1
    Debug=(bDrawDebug=true,bDraw2D=False,bDrawLocalAxes=False,DebugScale=25,DebugSegments=20,DebugThickness=0.25)
    DriverAxis=(X=0,Y=1,Z=0)
    DriverItem=(Type=Bone,Name="head")
    FalloffRegionScaleFactors=(PositiveWidth=1,NegativeWidth=0.5,PositiveHeight=1,NegativeHeight=1)
    FalloffSize=0.3
    OptionalParentItem=(Type=Bone,Name="None")
    RotationOffset=(X=90,Y=0,Z=-31)
    ExecutePin=@RigVMFunction_Sequence_3.A
   )
   > &RigUnit_ModifyTransforms(
    ItemToModify=((Item=(Type=Bone,Name="dyn_hair_side_01_r"),Transform=(Rotation=(X=0,Y=0,Z=0,W=1),Translation=(X=0,Y=-3,Z=0),Scale3D=(X=1,Y=1,Z=1))))
    Weight=@n3.OutputParam
   )
   > &RigUnit_ModifyTransforms(
    ItemToModify=((Item=(Type=Bone,Name="dyn_hair_side_02_r"),Transform=(Rotation=(X=0,Y=0,Z=0,W=1),Translation=(X=-2,Y=-2,Z=0),Scale3D=(X=1,Y=1,Z=1))))
    Weight=@n3.OutputParam
   )
   n4 = &RigUnit_SphericalPoseReader(
    ActiveRegionScaleFactors=(PositiveWidth=1,NegativeWidth=1,PositiveHeight=1,NegativeHeight=1)
    ActiveRegionSize=0.05
    Debug=(bDrawDebug=true,bDraw2D=False,bDrawLocalAxes=False,DebugScale=25,DebugSegments=20,DebugThickness=0.25)
    DriverAxis=(X=-1,Y=0,Z=0)
    DriverItem=(Type=Bone,Name="clavicle_r")
    FalloffRegionScaleFactors=(PositiveWidth=1,NegativeWidth=1,PositiveHeight=1,NegativeHeight=1)
    FalloffSize=0.2
    OptionalParentItem=(Type=Bone)
    RotationOffset=(X=110,Y=-20,Z=80)
    ExecutePin=@RigVMFunction_Sequence_3.B
   )
   n5 = &RigVMFunction_MathDoubleMul(B=@n4.OutputParam, A=@n3.OutputParam)
   > &RigUnit_ModifyTransforms(
    ItemToModify=((Item=(Type=Bone,Name="dyn_hair_side_01_r"),Transform=(Rotation=(X=0,Y=0,Z=0,W=1),Translation=(X=-1,Y=3,Z=0),Scale3D=(X=1,Y=1,Z=1))))
    Weight=@n5.Result
   )
   > &RigUnit_ModifyTransforms(
    ItemToModify=((Item=(Type=Bone,Name="dyn_hair_side_02_r"),Transform=(Rotation=(X=0,Y=0,Z=0,W=1),Translation=(X=1,Y=2,Z=0),Scale3D=(X=1,Y=1,Z=1))))
    Weight=@n5.Result
   )
   n6 = &RigUnit_SphericalPoseReader(
    ActiveRegionScaleFactors=(PositiveWidth=1,NegativeWidth=1,PositiveHeight=1,NegativeHeight=1)
    ActiveRegionSize=0.1
    Debug=(bDrawDebug=true,bDraw2D=False,bDrawLocalAxes=False,DebugScale=25,DebugSegments=20,DebugThickness=0.25)
    DriverAxis=(X=-1,Y=0,Z=0)
    DriverItem=(Type=Bone,Name="upperarm_r")
    FalloffRegionScaleFactors=(PositiveWidth=1,NegativeWidth=1,PositiveHeight=1,NegativeHeight=1)
    FalloffSize=0.3
    OptionalParentItem=(Type=Bone,Name="None")
    RotationOffset=(X=-112,Y=4,Z=-37)
    ExecutePin=@RigVMFunction_Sequence_3.C
   )
   > &RigUnit_ModifyTransforms(
    ItemToModify=((Item=(Type=Bone,Name="dyn_extra_upperarm_r"),Transform=(Rotation=(X=0,Y=0,Z=0,W=1),Translation=(X=0,Y=0,Z=5),Scale3D=(X=1,Y=1,Z=1))))
    Weight=@n6.OutputParam
   )
   RigUnit_GetTransform = &RigUnit_GetTransform(Item=(Type=Bone,Name="dyn_hair_side_01_r"), bInitial=False)
   RigVMFunction_MathVectorDistance = &RigVMFunction_MathVectorDistance(
    A=@RigUnit_GetTransform.Transform.Translation
    B=&RigUnit_GetTransform(Item=(Type=Bone,Name="head"), bInitial=False).Transform.Translation
   )
   n7 = &RigVMFunction_AlphaInterp(
    InRange=(Min=9,Max=9.5)
    InterpSpeedDecreasing=8
    InterpSpeedIncreasing=8
    OutRange=(Min=1,Max=0)
    Scale=1
    bClampResult=true
    bMapRange=true
    Value=@RigVMFunction_MathVectorDistance.Result
   )
   > &RigUnit_ModifyTransforms(
    ItemToModify=((Item=(Type=Bone,Name="dyn_hair_side_01_r"),Transform=(Rotation=(X=0,Y=0,Z=0,W=1),Translation=(X=3,Y=3,Z=3),Scale3D=(X=1,Y=1,Z=1))))
    Weight=@n7.Result
   )
   n8 = &RigUnit_SphericalPoseReader(
    ActiveRegionScaleFactors=(PositiveWidth=1,NegativeWidth=1,PositiveHeight=0.75,NegativeHeight=0.75)
    ActiveRegionSize=0.05
    Debug=(bDrawDebug=true,bDraw2D=False,bDrawLocalAxes=False,DebugScale=25,DebugSegments=20,DebugThickness=0.25)
    DriverAxis=(X=0,Y=1,Z=0)
    DriverItem=(Type=Bone,Name="head")
    FalloffRegionScaleFactors=(PositiveWidth=1,NegativeWidth=0.5,PositiveHeight=0.5,NegativeHeight=0.5)
    FalloffSize=0.3
    OptionalParentItem=(Type=Bone,Name="neck_01")
    RotationOffset=(X=90,Y=0,Z=0)
    ExecutePin=@n2.A
   )
   n9 = &RigVMFunction_AlphaInterp(
    InterpSpeedDecreasing=8
    InterpSpeedIncreasing=8
    OutRange=(Min=1,Max=0.5)
    Scale=1
    bClampResult=False
    bMapRange=true
    Value=&RigVMFunction_MathBoolToFloat(Value=RigVMVariableNode() {variable=moving_2D}.Value).Result
   )
   n10 = &RigVMFunction_AlphaInterp(
    InterpSpeedDecreasing=8
    InterpSpeedIncreasing=8
    OutRange=(Min=1,Max=0.5)
    Scale=1
    bClampResult=False
    bMapRange=true
    Value=&RigVMFunction_MathBoolToFloat(Value=RigVMVariableNode() {variable=aim}.Value).Result
   )
   n11 = &RigVMFunction_AlphaInterp(
    InterpSpeedDecreasing=4
    InterpSpeedIncreasing=2
    OutRange=(Min=1,Max=0.5)
    Scale=1
    bClampResult=False
    bMapRange=false
    Value=@n8.OutputParam
   )
   Multiply_2 = RigVMAggregateNode(B=@n9.Result, C=@n10.Result, A=@n11.Result)
   RigVMFunction_MathBoolToFloat = &RigVMFunction_MathBoolToFloat(Value=RigVMVariableNode() {variable=FE_idel}.Value)
   Maximum = &RigVMFunction_MathDoubleMax(A=@Multiply_2.Result, B=@RigVMFunction_MathBoolToFloat.Result)
   > &RigUnit_ModifyTransforms(
    ItemToModify=((Item=(Type=Bone,Name="dyn_hair_side_bk_01_r"),Transform=(Rotation=(X=0,Y=-0.1736,Z=0,W=0.9848),Translation=(X=0,Y=0,Z=0),Scale3D=(X=1,Y=1,Z=1))))
    Weight=@Maximum.Result
   )
   > &RigUnit_ModifyTransforms(
    ItemToModify=((Item=(Type=Bone,Name="dyn_hair_side_bk_02_r"),Transform=(Rotation=(X=0,Y=0,Z=0,W=1),Translation=(X=0,Y=0,Z=0),Scale3D=(X=1,Y=1,Z=1))))
    Weight=@Maximum.Result
   )
   > &RigUnit_ModifyTransforms(
    ItemToModify=((Item=(Type=Bone,Name="dyn_hair_side_bk_03_r"),Transform=(Rotation=(X=-0.1634,Y=0.1783,Z=-0.0122,W=0.9702),Translation=(X=0,Y=3,Z=1),Scale3D=(X=1,Y=1,Z=1))))
    Weight=@Maximum.Result
   )
   > &RigUnit_ModifyTransforms(
    ItemToModify=((Item=(Type=Bone,Name="dyn_hair_side_02_r"),Transform=(Rotation=(X=0.0076,Y=0.1735,Z=-0.1285,W=0.9764),Translation=(X=-2,Y=0,Z=-1),Scale3D=(X=1,Y=1,Z=1))))
    Weight=@Maximum.Result
   )
   > &RigUnit_ModifyTransforms(
    ItemToModify=((Item=(Type=Bone,Name="dyn_hair_side_bk_01_l"),Transform=(Rotation=(X=0,Y=0.0872,Z=0,W=0.9962),Translation=(X=0,Y=0,Z=0),Scale3D=(X=1,Y=1,Z=1))))
    Weight=@Maximum.Result
   )
   > &RigUnit_ModifyTransforms(
    ItemToModify=((Item=(Type=Bone,Name="dyn_hair_side_bk_02_l"),Transform=(Rotation=(X=0,Y=-0.0872,Z=0,W=0.9962),Translation=(X=0,Y=0,Z=0),Scale3D=(X=1,Y=1,Z=1))))
    Weight=@Maximum.Result
   )
   > &RigUnit_ModifyTransforms(
    ItemToModify=((Item=(Type=Bone,Name="dyn_hair_center_bk_01"),Transform=(Rotation=(X=0,Y=0,Z=0.1305,W=0.9914),Translation=(X=0,Y=0,Z=0),Scale3D=(X=1,Y=1,Z=1))))
    Weight=@Maximum.Result
   )
   > &RigUnit_ModifyTransforms(
    ItemToModify=((Item=(Type=Bone,Name="dyn_hair_center_bk_02"),Transform=(Rotation=(X=0,Y=0,Z=-0.1736,W=0.9848),Translation=(X=0,Y=0,Z=0),Scale3D=(X=1,Y=1,Z=1))))
    Weight=@Maximum.Result
   )
   > &RigUnit_ModifyTransforms(
    ItemToModify=((Item=(Type=Bone,Name="dyn_hair_side_01_l"),Transform=(Rotation=(X=-0.0057,Y=0.1304,Z=0.0432,W=0.9905),Translation=(X=0,Y=0,Z=0),Scale3D=(X=1,Y=1,Z=1))))
    Weight=@Maximum.Result
   )
   > &RigUnit_ModifyTransforms(
    ItemToModify=((Item=(Type=Bone,Name="dyn_hair_side_bk_l_base"),Transform=(Rotation=(X=-0.0076,Y=0.0868,Z=0.0868,W=0.9924),Translation=(X=0,Y=0,Z=0),Scale3D=(X=1,Y=1,Z=1))))
    Weight=@Maximum.Result
   )
   > &RigUnit_ModifyTransforms(
    ItemToModify=((Item=(Type=Bone,Name="dyn_hair_side_bk_03_l"),Transform=(Rotation=(X=-0.0226,Y=-0.0842,Z=-0.2578,W=0.9623),Translation=(X=0,Y=0,Z=0),Scale3D=(X=1,Y=1,Z=1))))
    Weight=@Maximum.Result
   )
   > &RigUnit_ModifyTransforms(
    ItemToModify=((Item=(Type=Bone,Name="dyn_hair_side_fr_01_r"),Transform=(Rotation=(X=-0.0396,Y=-0.0472,Z=0.0851,W=0.9945),Translation=(X=0,Y=0,Z=0),Scale3D=(X=1,Y=1,Z=1))))
    Weight=&RigVMFunction_MathBoolToFloat(Value=RigVMVariableNode() {variable=FE_idel}.Value).Result
   )
  ]}
]
hierarchy [
 root {loc=<0 0 0>, rot=<0 0 0 1>}
 pelvis {parent=root, loc=<0 -0.5301 98.0571>, rot=<0.0127 0.707 -0.0127 -0.707>}
 spine_01 {parent=pelvis, loc=<6.1208 -0.9328 0>, rot=<0 0 0.034 -0.9994>}
 spine_02 {parent=spine_01, loc=<6.5997 -0.2095 0>, rot=<0 0 0.0238 -0.9997>}
 spine_03 {parent=spine_02, loc=<6.4914 -0.1237 0>, rot=<0 0 0.0149 -0.9999>}
 spine_04 {parent=spine_03, loc=<6.9598 0.0794 0>, rot=<0 0 -0.0132 -0.9999>}
 spine_05 {parent=spine_04, loc=<6.9987 0 0>, rot=<0 0 -0.0015 -1>}
 neck_01 {parent=spine_05, loc=<11.3138 0.1529 0>, rot=<0 0 -0.2003 -0.9797>}
 neck_02 {parent=neck_01, loc=<4.5627 0 0>, rot=<0 0 0.0504 -0.9987>}
 head {parent=neck_02, loc=<5.3914 0 0>, rot=<0 0 0.115 -0.9934>}
 dyn_hair_side_fr_l_base {parent=head, loc=<6.4868 3.7473 -7.2093>, rot=<0.0267 -0.9982 -0.0529 0.0014>}
 dyn_hair_side_fr_01_l {parent=dyn_hair_side_fr_l_base, loc=<0.481 0.0217 0.0023>, rot=<0 0 0 1>}
 dyn_hair_side_fr_02_l {parent=dyn_hair_side_fr_01_l, loc=<5.2398 0 0>, rot=<0.0084 0.0026 -0.0225 -0.9997>}
 dyn_hair_side_fr_03_l {parent=dyn_hair_side_fr_02_l, loc=<5.3735 0 0>, rot=<-0.0018 -0.0164 -0.002 -0.9999>}
 dyn_hair_side_fr_r_base {parent=head, loc=<6.487 3.7473 7.2093>, rot=<-0.0014 0.0529 0.9982 -0.0267>}
 dyn_hair_side_fr_01_r {parent=dyn_hair_side_fr_r_base, loc=<0.4815 -0.0217 0.0023>, rot=<0 0 0 1>}
 dyn_hair_side_fr_02_r {parent=dyn_hair_side_fr_01_r, loc=<5.2394 0 0>, rot=<-0.0084 0.0026 0.0225 -0.9997>}
 dyn_hair_side_fr_03_r {parent=dyn_hair_side_fr_02_r, loc=<5.374 0 0>, rot=<0.0018 -0.0164 0.002 -0.9999>}
 dyn_hair_side_l_base {parent=head, loc=<6.9421 -1.3141 -6.9293>, rot=<0.089 -0.9854 0.1379 -0.0453>}
 dyn_hair_side_01_l {parent=dyn_hair_side_l_base, loc=<0.481 0.0217 0.0023>, rot=<0 0 0 1>}
 dyn_hair_side_02_l {parent=dyn_hair_side_01_l, loc=<5.2398 0 0>, rot=<0.0117 0.003 0.0521 -0.9986>}
 dyn_hair_side_03_l {parent=dyn_hair_side_02_l, loc=<4.7514 0.1387 -0.0106>, rot=<-0.0008 -0.0165 0.0577 -0.9982>}
 dyn_hair_side_r_base {parent=head, loc=<6.9419 -1.3141 6.9293>, rot=<0.0453 -0.1379 0.9854 -0.089>}
 dyn_hair_side_01_r {parent=dyn_hair_side_r_base, loc=<0.4808 -0.0217 0.0023>, rot=<0 0 0 1>}
 dyn_hair_side_02_r {parent=dyn_hair_side_01_r, loc=<5.2396 0 0>, rot=<-0.0117 0.003 -0.0521 -0.9986>}
 dyn_hair_side_03_r {parent=dyn_hair_side_02_r, loc=<4.7513 -0.1387 -0.0106>, rot=<0.0008 -0.0165 -0.0577 -0.9982>}
 dyn_hair_side_bk_l_base {parent=head, loc=<7.1302 -5.0187 -4.6295>, rot=<0.0957 -0.9954 0 0>}
 dyn_hair_side_bk_01_l {parent=dyn_hair_side_bk_l_base, loc=<0.3444 0.0639 0>, rot=<0 0 0 1>}
 dyn_hair_side_bk_02_l {parent=dyn_hair_side_bk_01_l, loc=<7.7274 -0.3037 0.6968>, rot=<0.0588 0.0147 0.0601 -0.9964>}
 dyn_hair_side_bk_03_l {parent=dyn_hair_side_bk_02_l, loc=<8.1625 0.2252 0.0629>, rot=<-0.0202 -0.0129 0.0049 -0.9997>}
 dyn_hair_side_bk_04_l {parent=dyn_hair_side_bk_03_l, loc=<6.8219 0.0225 0.0281>, rot=<0.0057 -0.0208 -0.085 -0.9961>}
 dyn_hair_side_bk_05_l {parent=dyn_hair_side_bk_04_l, loc=<6.4095 0.0823 0.0318>, rot=<0.013 -0.0006 -0.0443 -0.9989>}
 dyn_hair_center_bk_base {parent=head, loc=<7.2097 -6.3219 0.3833>, rot=<0.144 -0.9894 0.0087 0.0139>}
 dyn_hair_center_bk_01 {parent=dyn_hair_center_bk_base, loc=<0.4251 0.1227 0.0111>, rot=<0 0 0 1>}
 dyn_hair_center_bk_02 {parent=dyn_hair_center_bk_01, loc=<7.9201 0.5099 0.2304>, rot=<0.0067 0.0149 -0.0053 -0.9999>}
 dyn_hair_center_bk_03 {parent=dyn_hair_center_bk_02, loc=<8.2378 0 0>, rot=<0 0 0.0023 -1>}
 dyn_hair_center_bk_04 {parent=dyn_hair_center_bk_03, loc=<8.2799 0 0>, rot=<0 0 -0.1277 -0.9918>}
 dyn_hair_center_bk_05 {parent=dyn_hair_center_bk_04, loc=<9.6093 0 0>, rot=<0 0 0 1>}
 dyn_hair_side_bk_r_base {parent=head, loc=<7.1305 -5.0187 4.6295>, rot=<0.0957 -0.9954 0 0>}
 dyn_hair_side_bk_01_r {parent=dyn_hair_side_bk_r_base, loc=<0.3451 0.064 0>, rot=<0 0 0 1>}
 dyn_hair_side_bk_02_r {parent=dyn_hair_side_bk_01_r, loc=<7.727 -0.3038 -0.6969>, rot=<-0.0588 -0.0147 0.0601 -0.9964>}
 dyn_hair_side_bk_03_r {parent=dyn_hair_side_bk_02_r, loc=<8.1626 0.2253 -0.0629>, rot=<0.0202 0.0129 0.0049 -0.9997>}
 dyn_hair_side_bk_04_r {parent=dyn_hair_side_bk_03_r, loc=<6.8218 0.0224 -0.0281>, rot=<-0.0057 0.0208 -0.085 -0.9961>}
 dyn_hair_side_bk_05_r {parent=dyn_hair_side_bk_04_r, loc=<6.4092 0.0823 -0.0318>, rot=<-0.013 0.0006 -0.0443 -0.9989>}
 dyn_hair_center_fr_l_base {parent=head, loc=<11.7785 6.4601 -2.3374>, rot=<-0.1077 -0.8105 -0.4996 -0.2863>}
 dyn_hair_center_fr_01_l {parent=dyn_hair_center_fr_l_base, loc=<0.3251 0 0.0956>, rot=<0 0 0 1>}
 dyn_hair_center_fr_02_l {parent=dyn_hair_center_fr_01_l, loc=<3.4498 0 0>, rot=<-0.0106 -0.2305 -0.0306 -0.9725>}
 dyn_hair_center_fr_base {parent=head, loc=<11.7748 6.9179 0.0568>, rot=<0.2092 0.6754 0.6754 0.2092>}
 dyn_hair_center_fr_01 {parent=dyn_hair_center_fr_base, loc=<0.3251 0 0.0956>, rot=<0 0 0 1>}
 dyn_hair_center_fr_02 {parent=dyn_hair_center_fr_01, loc=<3.4498 0 0>, rot=<0 -0.2337 0 -0.9723>}
 dyn_hair_center_fr_r_base {parent=head, loc=<11.7788 6.4601 2.3373>, rot=<0.2863 0.4996 0.8105 0.1077>}
 dyn_hair_center_fr_01_r {parent=dyn_hair_center_fr_r_base, loc=<0.3253 0 0.0954>, rot=<0 0 0 -1>}
 dyn_hair_center_fr_02_r {parent=dyn_hair_center_fr_01_r, loc=<3.4499 0 0>, rot=<0.0119 -0.2334 0.0497 -0.971>}
 extra_head {parent=head, loc=<0 -0.5415 0>, rot=<0 0 0 1>}
 clavicle_l {parent=spine_05, loc=<5.4831 -0.1053 -3.2946>, rot=<0.0465 -0.7289 -0.009 -0.6829>}
 upperarm_l {parent=clavicle_l, loc=<9.8658 -1.2136 -0.4105>, rot=<0.0164 -0.3245 -0.0269 -0.9454>}
 dyn_extra_upperarm_l {parent=upperarm_l, loc=<0.5 0 0>, rot=<0 0 0 1>}
 clavicle_r {parent=spine_05, loc=<5.4831 -0.1053 3.2946>, rot=<0.7289 0.0465 -0.6829 0.009>}
 upperarm_r {parent=clavicle_r, loc=<-9.8658 1.2136 0.4105>, rot=<0.0164 -0.3245 -0.0269 -0.9454>}
 dyn_extra_upperarm_r {parent=upperarm_r, loc=<-0.5 0 0>, rot=<0 0 0 1>}
 dyn_backpack_collision {parent=spine_05, loc=<10.9743 -22.8145 0>, rot=<0.0616 -0.7044 -0.0616 -0.7044>}
 SPR_HeadUp {type=Curve, loc=<0 0 0>, rot=<0 0 0 1>}
]
variables [
 FE_idel {type=bool, default=False}
 aim {type=bool, default=False}
 crouch {type=bool, default=False}
 moving_2D {type=bool, default=False}
]
