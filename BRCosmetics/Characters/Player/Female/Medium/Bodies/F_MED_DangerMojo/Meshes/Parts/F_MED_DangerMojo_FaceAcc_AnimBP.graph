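# Character-part AnimBP for the DangerMojo face accessory mesh, driving its dynamic hair.
# It layers body-mesh arms onto the parent's copied pose, then applies a Control Rig and bone corrections before RigidBody simulation.
# Blendable weighted constraints pull back-hair bones toward torso and head bones, followed by a post Control Rig and head-relative AnimDynamics for front hair.
# The EventGraph updates backbling, movement, aiming, skydive, front-end and LOD states, plus jump-adjusted gravity.
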AnimBlueprint /BRCosmetics/Characters/Player/Female/Medium/Bodies/F_MED_DangerMojo/Meshes/Parts/F_MED_DangerMojo_FaceAcc_AnimBP
target_skeleton=&/F_MED_DangerMojo_FaceAcc_Skeleton
graphs [
 AnimGraph {nodes [
   Local To Component()
   n1 = "Constraint - dyn_hair_side_bk_03_r (PPPPPPPP)"!(
    ComponentPose=Local To Component(LocalPose=Use cached pose 'post_dyn'())
    Alpha=0.3
    ConstraintWeights_0=0.589
    ConstraintWeights_1:float=0.295
    ConstraintWeights_2:float=0.042
    ConstraintWeights_3:float=0.0418
    ConstraintWeights_4:float=0.0136
    ConstraintWeights_5:float=0.0096
    ConstraintWeights_6:float=0.005
    ConstraintWeights_7:float=0.004
   ) -> (Pose) {
    alphaBoolBlend {blendOption=Linear}
    boneToModify {boneName=dyn_hair_side_bk_03_r}
    constraintSetup [
     {targetBone {boneName=spine_05}}
     {targetBone {boneName=clavicle_r}}
     {targetBone {boneName=neck_01}}
     {targetBone {boneName=neck_02}}
     {targetBone {boneName=spine_03}}
     {targetBone {boneName=clavicle_l}}
     {targetBone {boneName=head}}
     {targetBone {boneName=upperarm_r}}
    ]
    constraintWeights [1,1,1,1,1,1,1,1]
   }
   n2 = "Constraint - dyn_hair_side_bk_04_r (PPPPPPP)"!(
    ComponentPose=@n1
    Alpha=0.5
    ConstraintWeights_0=0.4848
    ConstraintWeights_1:float=0.3985
    ConstraintWeights_2:float=0.0636
    ConstraintWeights_3:float=0.0326
    ConstraintWeights_4:float=0.0203
    ConstraintWeights_5:float=0.0002
    ConstraintWeights_6:float=0.0001
   ) -> (Pose) {
    alphaBoolBlend {blendOption=Linear}
    boneToModify {boneName=dyn_hair_side_bk_04_r}
    constraintSetup [
     {targetBone {boneName=clavicle_r}}
     {targetBone {boneName=spine_05}}
     {targetBone {boneName=upperarm_r}}
     {targetBone {boneName=spine_04}}
     {targetBone {boneName=clavicle_l}}
     {targetBone {boneName=spine_03}}
     {targetBone {boneName=spine_02}}
    ]
    constraintWeights [1,1,1,1,1,1,1]
   }
   n3 = "Constraint - dyn_hair_side_bk_05_r (PPPPPPP)"!(
    ComponentPose=@n2
    Alpha=1
    ConstraintWeights_0=0.4066
    ConstraintWeights_1:float=0.39
    ConstraintWeights_2:float=0.1377
    ConstraintWeights_3:float=0.0302
    ConstraintWeights_4:float=0.026
    ConstraintWeights_5:float=0.0052
    ConstraintWeights_6:float=0.0042
   ) -> (Pose) {
    alphaBoolBlend {blendOption=Linear}
    boneToModify {boneName=dyn_hair_side_bk_05_r}
    constraintSetup [
     {targetBone {boneName=spine_05}}
     {targetBone {boneName=clavicle_r}}
     {targetBone {boneName=spine_04}}
     {targetBone {boneName=clavicle_l}}
     {targetBone {boneName=upperarm_r}}
     {targetBone {boneName=spine_03}}
     {targetBone {boneName=spine_02}}
    ]
    constraintWeights [1,1,1,1,1,1,1]
   }
   n4 = "Constraint - dyn_hair_center_bk_03 (PPPPPPP)"!(
    ComponentPose=@n3
    Alpha=0.3
    ConstraintWeights_0=0.7377
    ConstraintWeights_1:float=0.0658
    ConstraintWeights_2:float=0.064
    ConstraintWeights_3:float=0.0616
    ConstraintWeights_4:float=0.0356
    ConstraintWeights_5:float=0.0239
    ConstraintWeights_6:float=0.0114
   ) -> (Pose) {
    alphaBoolBlend {blendOption=Linear}
    boneToModify {boneName=dyn_hair_center_bk_03}
    constraintSetup [
     {targetBone {boneName=spine_05}}
     {targetBone {boneName=neck_02}}
     {targetBone {boneName=clavicle_r}}
     {targetBone {boneName=neck_01}}
     {targetBone {boneName=clavicle_l}}
     {targetBone {boneName=spine_03}}
     {targetBone {boneName=head}}
    ]
    constraintWeights [1,1,1,1,1,1,1]
   }
   n5 = "Constraint - dyn_hair_center_bk_04 (PPPPP)"!(
    ComponentPose=@n4
    Alpha=0.5
    ConstraintWeights_0=0.6234
    ConstraintWeights_1:float=0.2284
    ConstraintWeights_2:float=0.1165
    ConstraintWeights_3:float=0.0303
    ConstraintWeights_4:float=0.0014
   ) -> (Pose) {
    alphaBoolBlend {blendOption=Linear}
    boneToModify {boneName=dyn_hair_center_bk_04}
    constraintSetup [
     {targetBone {boneName=spine_05}}
     {targetBone {boneName=clavicle_r}}
     {targetBone {boneName=clavicle_l}}
     {targetBone {boneName=spine_04}}
     {targetBone {boneName=upperarm_r}}
    ]
    constraintWeights [1,1,1,1,1]
   }
   n6 = "Constraint - dyn_hair_center_bk_05 (PPPPPPP)"!(
    ComponentPose=@n5
    Alpha=1
    ConstraintWeights_0=0.3612
    ConstraintWeights_1:float=0.2668
    ConstraintWeights_2:float=0.1185
    ConstraintWeights_3:float=0.0868
    ConstraintWeights_4:float=0.0734
    ConstraintWeights_5:float=0.071
    ConstraintWeights_6:float=0.0222
   ) -> (Pose) {
    alphaBoolBlend {blendOption=Linear}
    boneToModify {boneName=dyn_hair_center_bk_05}
    constraintSetup [
     {targetBone {boneName=spine_04}}
     {targetBone {boneName=spine_05}}
     {targetBone {boneName=clavicle_r}}
     {targetBone {boneName=clavicle_l}}
     {targetBone {boneName=spine_02}}
     {targetBone {boneName=spine_03}}
     {targetBone {boneName=spine_01}}
    ]
    constraintWeights [1,1,1,1,1,1,1]
   }
   n7 = "Constraint - dyn_hair_side_bk_03_l (PPPPPPPP)"!(
    ComponentPose=@n6
    Alpha=0.3
    ConstraintWeights_0=0.5367
    ConstraintWeights_1:float=0.3514
    ConstraintWeights_2:float=0.0396
    ConstraintWeights_3:float=0.0394
    ConstraintWeights_4:float=0.0132
    ConstraintWeights_5:float=0.0121
    ConstraintWeights_6:float=0.0056
    ConstraintWeights_7:float=0.002
   ) -> (Pose) {
    alphaBoolBlend {blendOption=Linear}
    boneToModify {boneName=dyn_hair_side_bk_03_l}
    constraintSetup [
     {targetBone {boneName=spine_05}}
     {targetBone {boneName=clavicle_l}}
     {targetBone {boneName=neck_01}}
     {targetBone {boneName=neck_02}}
     {targetBone {boneName=upperarm_l}}
     {targetBone {boneName=spine_03}}
     {targetBone {boneName=clavicle_r}}
     {targetBone {boneName=head}}
    ]
    constraintWeights [1,1,1,1,1,1,1,1]
   }
   n8 = "Constraint - dyn_hair_side_bk_04_l (PPPPP)"!(
    ComponentPose=@n7
    Alpha=0.5
    ConstraintWeights_0=0.4778
    ConstraintWeights_1:float=0.429
    ConstraintWeights_2:float=0.0347
    ConstraintWeights_3:float=0.0294
    ConstraintWeights_4:float=0.029
   ) -> (Pose) {
    alphaBoolBlend {blendOption=Linear}
    boneToModify {boneName=dyn_hair_side_bk_04_l}
    constraintSetup [
     {targetBone {boneName=clavicle_l}}
     {targetBone {boneName=spine_05}}
     {targetBone {boneName=upperarm_l}}
     {targetBone {boneName=spine_04}}
     {targetBone {boneName=clavicle_r}}
    ]
    constraintWeights [1,1,1,1,1]
   }
   n9 = "Constraint - dyn_hair_side_bk_05_l (PPPPPPP)"!(
    ComponentPose=@n8
    Alpha=1
    ConstraintWeights_0=0.4201
    ConstraintWeights_1:float=0.3813
    ConstraintWeights_2:float=0.1341
    ConstraintWeights_3:float=0.0364
    ConstraintWeights_4:float=0.0209
    ConstraintWeights_5:float=0.004
    ConstraintWeights_6:float=0.0033
   ) -> (Pose) {
    alphaBoolBlend {blendOption=Linear}
    boneToModify {boneName=dyn_hair_side_bk_05_l}
    constraintSetup [
     {targetBone {boneName=spine_05}}
     {targetBone {boneName=clavicle_l}}
     {targetBone {boneName=spine_04}}
     {targetBone {boneName=clavicle_r}}
     {targetBone {boneName=upperarm_l}}
     {targetBone {boneName=spine_03}}
     {targetBone {boneName=spine_02}}
    ]
    constraintWeights [1,1,1,1,1,1,1]
   }
   n10 = Control Rig!(
    Source=Blend Poses by bool(BlendPose_0=Component To Local(ComponentPose=@n9), BlendPose_1=Use cached pose 'post_dyn'(), BlendTime_0=0.2, BlendTime_1=0.2) {blendType=Linear}
    Alpha
    crouch:bool=false
    aim:bool=Get Is Targeting() -> (Is Targeting:bool)
    moving_2D:bool=Get Is Moving 2D() -> (Is Moving 2D:bool)
    FE_idel:bool=Get FE_idel() -> (FE_idel:bool)
   ) -> (Pose) {
    alphaBoolBlend {blendOption=Linear}
    controlRigAssetReference {blueprintRigClass=&/F_MED_DangerMojo_FaceAcc_post_CtrlRig.F_MED_DangerMojo_FaceAcc_post_CtrlRig_C}
   }
   n11 = Anim Dynamics - Bone: dyn_hair_center_fr_01_l(ComponentPose=Local To Component(LocalPose=@n10), Alpha=1, ExternalForce="0,0,0") {
    angularDampingOverride=1
    angularSpringConstant=20000
    bAngularSpring=true
    bOverrideAngularDamping=true
    bUseGravityOverride=true
    gravityScale=0.35
    linearSpringConstant=5000
    simulationSpace=BoneRelative
    alphaBoolBlend {blendOption=Linear}
    boundBone {boneName=dyn_hair_center_fr_01_l}
    componentLinearAccScale=<1 1 1>
    componentLinearVelScale=<1 1 1>
    gravityOverride {z=-980}
    physicsBodyDefinitions [{
      boundBone {boneName=dyn_hair_center_fr_01_l}
      boxExtents=<20 30 30>
      constraintSetup {angularLimitsMax=<15 0 15>, angularLimitsMin=<-15 -15 -15>, angularTarget=<1 0 0>}
      localJointOffset=<4 0 0>
     }]
    relativeSpaceBone {boneName=head}
    sphericalLimits [{limitRadius=500, drivingBone {boneName=head}, sphereLocalOffset=<4.7757 0.4962 0.1794>}]
   }
   n12 = Anim Dynamics - Bone: dyn_hair_center_fr_02_l(ComponentPose=@n11, Alpha=1, ExternalForce="0,0,0") {
    angularDampingOverride=1
    angularSpringConstant=20000
    bAngularSpring=true
    bOverrideAngularDamping=true
    bUseGravityOverride=true
    gravityScale=0.35
    linearSpringConstant=5000
    simulationSpace=BoneRelative
    alphaBoolBlend {blendOption=Linear}
    boundBone {boneName=dyn_hair_center_fr_02_l}
    componentLinearAccScale=<1 1 1>
    componentLinearVelScale=<1 1 1>
    gravityOverride {z=-980}
    physicsBodyDefinitions [{
      boundBone {boneName=dyn_hair_center_fr_02_l}
      boxExtents=<20 30 30>
      constraintSetup {angularLimitsMax=<15 0 15>, angularLimitsMin=<-15 -15 -15>, angularTarget=<1 0 0>}
      localJointOffset=<4 0 0>
     }]
    relativeSpaceBone {boneName=head}
    sphericalLimits [{limitRadius=5, drivingBone {boneName=head}, sphereLocalOffset=<0 0 0>}]
   }
   n13 = Anim Dynamics - Bone: dyn_hair_center_fr_01(ComponentPose=@n12, Alpha=1, ExternalForce="0,0,0") {
    angularDampingOverride=1
    angularSpringConstant=20000
    bAngularSpring=true
    bOverrideAngularDamping=true
    bUseGravityOverride=true
    gravityScale=0.35
    linearSpringConstant=5000
    simulationSpace=BoneRelative
    alphaBoolBlend {blendOption=Linear}
    boundBone {boneName=dyn_hair_center_fr_01}
    componentLinearAccScale=<1 1 1>
    componentLinearVelScale=<1 1 1>
    gravityOverride {z=-980}
    physicsBodyDefinitions [{
      boundBone {boneName=dyn_hair_center_fr_01}
      boxExtents=<20 30 30>
      constraintSetup {angularLimitsMax=<15 0 15>, angularLimitsMin=<-15 -15 -15>, angularTarget=<1 0 0>}
      localJointOffset=<4 0 0>
     }]
    relativeSpaceBone {boneName=head}
    sphericalLimits [{limitRadius=5, drivingBone {boneName=head}, sphereLocalOffset=<0 0 0>}]
   }
   n14 = Anim Dynamics - Bone: dyn_hair_center_fr_02(ComponentPose=@n13, Alpha=1, ExternalForce="0,0,0") {
    angularDampingOverride=1
    angularSpringConstant=20000
    bAngularSpring=true
    bOverrideAngularDamping=true
    bUseGravityOverride=true
    gravityScale=0.35
    linearSpringConstant=5000
    simulationSpace=BoneRelative
    alphaBoolBlend {blendOption=Linear}
    boundBone {boneName=dyn_hair_center_fr_02}
    componentLinearAccScale=<1 1 1>
    componentLinearVelScale=<1 1 1>
    gravityOverride {z=-980}
    physicsBodyDefinitions [{
      boundBone {boneName=dyn_hair_center_fr_02}
      boxExtents=<20 30 30>
      constraintSetup {angularLimitsMax=<15 0 15>, angularLimitsMin=<-15 -15 -15>, angularTarget=<1 0 0>}
      localJointOffset=<4 0 0>
     }]
    relativeSpaceBone {boneName=head}
    sphericalLimits [{limitRadius=5, drivingBone {boneName=head}, sphereLocalOffset=<0 0 0>}]
   }
   n15 = Anim Dynamics - Bone: dyn_hair_center_fr_01_r(ComponentPose=@n14, Alpha=1, ExternalForce="0,0,0") {
    angularDampingOverride=1
    angularSpringConstant=20000
    bAngularSpring=true
    bOverrideAngularDamping=true
    bUseGravityOverride=true
    gravityScale=0.35
    linearSpringConstant=5000
    simulationSpace=BoneRelative
    alphaBoolBlend {blendOption=Linear}
    boundBone {boneName=dyn_hair_center_fr_01_r}
    componentLinearAccScale=<1 1 1>
    componentLinearVelScale=<1 1 1>
    gravityOverride {z=-980}
    physicsBodyDefinitions [{
      boundBone {boneName=dyn_hair_center_fr_01_r}
      boxExtents=<20 30 30>
      constraintSetup {angularLimitsMax=<15 0 15>, angularLimitsMin=<-15 -15 -15>, angularTarget=<1 0 0>}
      localJointOffset=<4 0 0>
     }]
    relativeSpaceBone {boneName=head}
    sphericalLimits [{limitRadius=5, drivingBone {boneName=head}, sphereLocalOffset=<0 0 0>}]
   }
   n16 = Anim Dynamics - Bone: dyn_hair_center_fr_02_r(ComponentPose=@n15, Alpha=1, ExternalForce="0,0,0") {
    angularDampingOverride=1
    angularSpringConstant=20000
    bAngularSpring=true
    bOverrideAngularDamping=true
    bUseGravityOverride=true
    gravityScale=0.35
    linearSpringConstant=5000
    simulationSpace=BoneRelative
    alphaBoolBlend {blendOption=Linear}
    boundBone {boneName=dyn_hair_center_fr_02_r}
    componentLinearAccScale=<1 1 1>
    componentLinearVelScale=<1 1 1>
    gravityOverride {z=-980}
    physicsBodyDefinitions [{
      boundBone {boneName=dyn_hair_center_fr_02_r}
      boxExtents=<20 30 30>
      constraintSetup {angularLimitsMax=<15 0 15>, angularLimitsMin=<-15 -15 -15>, angularTarget=<1 0 0>}
      localJointOffset=<4 0 0>
     }]
    relativeSpaceBone {boneName=head}
    sphericalLimits [{limitRadius=5, drivingBone {boneName=head}, sphereLocalOffset=<0 0 0>}]
   }
   Output Pose(Result=Component To Local(ComponentPose=@n16)) {layerGroup=None}
   n17 = Layered blend per bone(
    BasePose=Copy Pose From Mesh() {bUseAttachedParent=true}
    BlendPoses_0=Copy Pose From Mesh(SourceMeshComponent=Get Body_Mesh()) {bUseAttachedParent=true}
   ) {layerSetup [{branchFilters [{boneName=clavicle_l}, {boneName=clavicle_r}, {boneName=upperarm_r}, {boneName=upperarm_l}]}]}
   n18 = Control Rig!(Source=@n17, Alpha, Crouch:bool=false) -> (Pose) {
    alphaBoolBlend {blendOption=Linear}
    controlRigAssetReference {blueprintRigClass=&/F_MED_DangerMojo_FaceAcc_CtrlRig.F_MED_DangerMojo_FaceAcc_CtrlRig_C}
   }
   n19 = "Transform (Modify) Bone - Bone: extra_head"(ComponentPose=Local To Component(LocalPose=@n18), Alpha=1, Translation="0,0,0", Rotation="0,0,0", Scale="1.1,1.1,1.1") {
    scaleMode=BMM_Additive
    scaleSpace=BCS_BoneSpace
    translationMode=BMM_Additive
    translationSpace=BCS_BoneSpace
    alphaBoolBlend {blendOption=Linear}
    boneToModify {boneName=extra_head}
    scale=<1.1 1.1 1.1>
   }
   n20 = "Transform (Modify) Bone - Bone: extra_head"(ComponentPose=@n19, Alpha=1, Translation="0,0,0", Rotation="0,0,0", Scale="1.1,1.1,1.1") {
    scaleMode=BMM_Additive
    scaleSpace=BCS_BoneSpace
    alphaBoolBlend {blendOption=Linear}
    boneToModify {boneName=extra_head}
    scale=<1.1 1.1 1.1>
   }
   n21 = "Constraint - dyn_hair_side_r_base (R)"(ComponentPose=@n20, Alpha=1, ConstraintWeights_0=1) {
    alphaBoolBlend {blendOption=Linear}
    boneToModify {boneName=dyn_hair_side_r_base}
    constraintSetup [{transformType=Rotation, targetBone {boneName=spine_05}}]
    constraintWeights [1]
   }
   n22 = Look At - Bone: dyn_hair_side_bk_l_base(ComponentPose=@n21, Alpha=1) {
    alphaBoolBlend {blendOption=Linear}
    boneToModify {boneName=dyn_hair_side_bk_l_base}
    lookAtLocation=<7.6519 -16.3529 -14.2504>
    lookAtTarget {boneReference {boneName=pelvis}}
    lookAt_Axis {axis=<1 0>}
    lookUp_Axis {axis=<0 1>}
   }
   n23 = Look At - Bone: dyn_hair_center_bk_base(ComponentPose=@n22, Alpha=1) {
    alphaBoolBlend {blendOption=Linear}
    boneToModify {boneName=dyn_hair_center_bk_base}
    lookAtLocation=<7.7045 -29.7768 2.1848>
    lookAtTarget {boneReference {boneName=pelvis}}
    lookAt_Axis {axis=<1 0>}
    lookUp_Axis {axis=<0 1>}
   }
   n24 = Look At - Bone: dyn_hair_side_bk_r_base(ComponentPose=@n23, Alpha=1) {
    alphaBoolBlend {blendOption=Linear}
    boneToModify {boneName=dyn_hair_side_bk_r_base}
    lookAtLocation=<1.5672 -4.1614 10.9371>
    lookAtTarget {boneReference {boneName=pelvis}}
    lookAt_Axis {axis=<1 0>}
    lookUp_Axis {axis=<0 1>}
   }
   n25 = "Constraint - dyn_hair_side_bk_r_base (P)"(ComponentPose=@n24, Alpha=0.5, ConstraintWeights_0=1) {
    alphaBoolBlend {blendOption=Linear}
    boneToModify {boneName=dyn_hair_side_bk_r_base}
    constraintSetup [{targetBone {boneName=dyn_hair_side_r_base}}]
    constraintWeights [1]
   }
   n26 = "Transform (Modify) Bone - Bone: dyn_hair_side_bk_r_base"(ComponentPose=@n25, Alpha=1, Translation="-4.943,-0.656,-5.983", Rotation="30.092,28.545,50.796", Scale="1,1,1") {
    rotationMode=BMM_Additive
    rotationSpace=BCS_BoneSpace
    scaleMode=BMM_Additive
    scaleSpace=BCS_BoneSpace
    translationMode=BMM_Additive
    translationSpace=BCS_BoneSpace
    alphaBoolBlend {blendOption=Linear}
    boneToModify {boneName=dyn_hair_side_bk_r_base}
    rotation=<rot 30.092 28.545 50.796>
    translation=<-4.943 -0.656 -5.983>
   }
   n27 = "Transform (Modify) Bone - Bone: dyn_hair_center_bk_base"(ComponentPose=@n26, Alpha=1, Translation="0.004,-1.312,0.830", Rotation="0,0,0", Scale="1,1,1") {
    rotationMode=BMM_Additive
    rotationSpace=BCS_BoneSpace
    scaleMode=BMM_Additive
    scaleSpace=BCS_BoneSpace
    translationMode=BMM_Additive
    translationSpace=BCS_BoneSpace
    alphaBoolBlend {blendOption=Linear}
    boneToModify {boneName=dyn_hair_center_bk_base}
    translation=<0.004 -1.312 0.83>
   }
   n28 = "Transform (Modify) Bone - Bone: dyn_hair_side_bk_l_base"(ComponentPose=@n27, Alpha=1, Translation="0.004,-2.018,1.310", Rotation="0,0,0", Scale="1,1,1") {
    rotationMode=BMM_Additive
    rotationSpace=BCS_BoneSpace
    scaleMode=BMM_Additive
    scaleSpace=BCS_BoneSpace
    translationMode=BMM_Additive
    translationSpace=BCS_BoneSpace
    alphaBoolBlend {blendOption=Linear}
    boneToModify {boneName=dyn_hair_side_bk_l_base}
    translation=<0.004 -2.018 1.31>
   }
   Save cached pose 'pre_dyn'(Pose=Component To Local(ComponentPose=@n28)) {cacheName=pre_dyn}
   n29 = "Transform (Modify) Bone - Bone: dyn_hair_center_bk_base"(ComponentPose=Local To Component(LocalPose=Use cached pose 'pre_dyn'()), Alpha=1, Translation="0,0,0", Rotation="0,30,0", Scale="1,1,1") {
    rotationMode=BMM_Additive
    rotationSpace=BCS_BoneSpace
    alphaBoolBlend {blendOption=Linear}
    boneToModify {boneName=dyn_hair_center_bk_base}
    rotation {yaw=30}
   }
   n30 = "Transform (Modify) Bone - Bone: dyn_hair_side_bk_l_base"(ComponentPose=@n29, Alpha=1, Translation="0,0,0", Rotation="0,30,0", Scale="1,1,1") {
    rotationMode=BMM_Additive
    rotationSpace=BCS_BoneSpace
    alphaBoolBlend {blendOption=Linear}
    boneToModify {boneName=dyn_hair_side_bk_l_base}
    rotation {yaw=30}
   }
   n31 = "Transform (Modify) Bone - Bone: dyn_hair_side_bk_r_base"(ComponentPose=@n30, Alpha=1, Translation="0,0,0", Rotation="0,30,0", Scale="1,1,1") {
    rotationMode=BMM_Additive
    rotationSpace=BCS_BoneSpace
    alphaBoolBlend {blendOption=Linear}
    boneToModify {boneName=dyn_hair_side_bk_r_base}
    rotation {yaw=30}
   }
   n32 = Blend Poses by bool(BlendPose_0=Component To Local(ComponentPose=@n31), BlendPose_1=Use cached pose 'pre_dyn'()) {blendType=Linear}
   n33 = "Constraint - dyn_hair_side_bk_l_base (PPP)"!(
    ComponentPose=Local To Component(LocalPose=Use cached pose 'pre_dyn'())
    Alpha=0.5
    ConstraintWeights_0=0.9995
    ConstraintWeights_1:float=0.0003
    ConstraintWeights_2:float=0.0001
   ) -> (Pose) {
    alphaBoolBlend {blendOption=Linear}
    boneToModify {boneName=dyn_hair_side_bk_l_base}
    constraintSetup [{targetBone {boneName=head}}, {targetBone {boneName=spine_05}}, {targetBone {boneName=neck_02}}]
    constraintWeights [1,1,1]
   }
   n34 = "Constraint - dyn_hair_center_bk_base (PPPPPPP)"!(
    ComponentPose=@n33
    Alpha=0.5
    ConstraintWeights_0=0.9756
    ConstraintWeights_1:float=0.0115
    ConstraintWeights_2:float=0.008
    ConstraintWeights_3:float=0.002
    ConstraintWeights_4:float=0.0012
    ConstraintWeights_5:float=0.0012
    ConstraintWeights_6:float=0.0005
   ) -> (Pose) {
    alphaBoolBlend {blendOption=Linear}
    boneToModify {boneName=dyn_hair_center_bk_base}
    constraintSetup [
     {targetBone {boneName=head}}
     {targetBone {boneName=spine_05}}
     {targetBone {boneName=neck_02}}
     {targetBone {boneName=neck_01}}
     {targetBone {boneName=clavicle_r}}
     {targetBone {boneName=clavicle_l}}
     {targetBone {boneName=spine_03}}
    ]
    constraintWeights [1,1,1,1,1,1,1]
   }
   n35 = "Transform (Modify) Bone - Bone: dyn_backpack_collision"(ComponentPose=@n34, Alpha=1, Translation="0,30,0", Rotation="0,0,-5", Scale="1,1,1") {
    rotationMode=BMM_Additive
    rotationSpace=BCS_BoneSpace
    translationMode=BMM_Additive
    translationSpace=BCS_BoneSpace
    alphaBoolBlend {blendOption=Linear}
    alphaScaleBiasClamp {bInterpResult=true}
    boneToModify {boneName=dyn_backpack_collision}
    rotation {roll=-5}
    translation {y=30}
   }
   n36 = "Transform (Modify) Bone - Bone: dyn_backpack_collision"!(ComponentPose=@n35, bAlphaBoolEnabled:bool=True, Translation="0,-8,0", Rotation="0,0,0", Scale="1,1,1") -> (Pose) {
    alphaInputType=Bool
    rotationMode=BMM_Additive
    rotationSpace=BCS_BoneSpace
    translationMode=BMM_Additive
    translationSpace=BCS_BoneSpace
    alphaBoolBlend {blendOption=Linear}
    alphaScaleBiasClamp {bInterpResult=true}
    boneToModify {boneName=dyn_backpack_collision}
    translation {y=-8}
   }
   n37 = Component To Local(ComponentPose=@n36)
   n38 = RigidBody!(
    ComponentPose=Local To Component(LocalPose=Blend Poses by bool(BlendPose_0=@n37, BlendPose_1=@n32, BlendTime_0=0.2, BlendTime_1=0.2) {blendType=Linear})
    Alpha
    OverrideWorldGravity:Vector="0,0,0"
    ExternalForce
    ComponentLinearAccScale:Vector="0.1,0.1,0.05"
    bTransferBoneVelocities:bool=True
   ) -> (Pose) {
    bClampLinearTranslationLimitToRefPose=true
    bFreezeIncomingPoseOnStart=true
    bOverrideWorldGravity=true
    bTransferBoneVelocities=true
    overridePhysicsAsset=&/F_MED_DangerMojo_FaceAcc_Physics
    simulationSpace=BaseBoneSpace
    alphaBoolBlend {blendOption=Linear}
    alphaScaleBiasClamp {interpSpeedDecreasing=4, interpSpeedIncreasing=6}
    baseBoneRef {boneName=neck_02}
    componentAppliedLinearAccClamp=<800 800 500>
    componentLinearAccScale=<0.6 0.6 0.2>
    componentLinearVelScale=<0.5 0.5 0.5>
    overrideWorldGravity {z=-875}
    simSpaceSettings {maxAngularAcceleration=200, maxAngularVelocity=5, maxLinearAcceleration=500, maxLinearVelocity=200, worldAlpha=0.5}
   }
   n39 = "Transform (Modify) Bone - Bone: dyn_hair_side_bk_03_r"(ComponentPose=@n38, Alpha=1, Translation="-0.311,0.022,-1.753", Rotation="-5.038,-0.009,-0.000", Scale="1,1,1") {
    rotationMode=BMM_Additive
    rotationSpace=BCS_BoneSpace
    scaleMode=BMM_Additive
    scaleSpace=BCS_BoneSpace
    translationMode=BMM_Additive
    translationSpace=BCS_BoneSpace
    alphaBoolBlend {blendOption=Linear}
    boneToModify {boneName=dyn_hair_side_bk_03_r}
    rotation=<rot -5.038 -0.0089 0>
    translation=<-0.311 0.022 -1.753>
   }
   n40 = "Constraint - dyn_hair_side_03_r (R)"(ComponentPose=@n39, Alpha=0.01, ConstraintWeights_0=1) {
    alphaBoolBlend {blendOption=Linear}
    boneToModify {boneName=dyn_hair_side_03_r}
    constraintSetup [{transformType=Rotation, targetBone {boneName=dyn_hair_side_bk_02_r}}]
    constraintWeights [1]
   }
   Save cached pose 'post_dyn'(Pose=Component To Local(ComponentPose=@n40)) {cacheName=post_dyn}
  ]}
 EventGraph {nodes [
   Event OnCharacterCustomizationLoaded()
   > n1 = Cast To PlayerPawn_Athena(Object=TryGetPawnOwner())
   > Set Body_Mesh(Body_Mesh=GetSkeletalMeshForPartType(self=@n1))
   Event BlueprintInitializeAnimation()
   > Set Skel Mesh(Skel Mesh=GetOwnerSkeletalMeshComponent())
   Event BlueprintUpdateAnimation()
   > n2 = Sequence!(execute) -> (then_0, then_1, then_2:Exec)
   > n3 = Sequence!(execute) -> (then_0, then_1)
   n4 = CheckForBackbling()
   > n5 = Branch(Condition=@n4)
   > Set Is Wearing Backbling(Is Wearing Backbling=true)
   Set Is Wearing Backbling(execute=@n5.else, Is Wearing Backbling=false)
   n6 = Branch(execute=@n3.then_1, Condition=@n4.bIsBackblingACape)
   > Set Is Backbling Cape(Is Backbling Cape:bool=true) -> (Output_Get:bool)
   Set Is Backbling Cape(execute=@n6.else, Is Backbling Cape=false)
   n7 = Sequence!(execute=@n2.then_1) -> (then_0, then_1, then_2:Exec, then_3:Exec, then_4:Exec)
   > Set SPR_HeadUp(SPR_HeadUp:float=GetCurveValue(CurveName=SPR_HeadUp)) -> (Output_Get:float)
   n8 = Get bIsParachuteOpened()
   Set skydive(execute=@n7.then_1, skydive:bool=AND Boolean(A=Get bIsSkyDiving(), B=NOT Boolean(A=@n8))) -> (Output_Get:bool)
   Set Is Targeting(execute=@n7.then_2, Is Targeting:bool=Get bIsTargeting()) -> (Output_Get:bool)
   Set scale_extra_head(execute=@n7.then_3, scale_extra_head:bool="integer >= integer"(A=Get CurrentLOD(), B=2)) -> (Output_Get:bool)
   > Set Is Moving 2D(Is Moving 2D:bool=Get bIsMoving2D()) -> (Output_Get:bool)
   n9 = Get bIsPlayingEmote()
   Set FE_idel(execute=@n7.then_4, FE_idel:bool=AND Boolean(A=NOT Boolean(A=@n9), B=Get bIsInFrontEnd())) -> (Output_Get:bool)
   n10 = Sequence!(execute=@n2.then_2) -> (then_0, then_1)
   n11 = Gravity Override State Function DEPRECATED(
    execute=@n10.then_1
    CharacterPartAnimInstance=Self-Reference()
    GameTimeInSeconds=GetGameTimeInSeconds()
    GravityOverrideParams=Get Gravity Override Params()
    DebugJointName=spine_05
   )
   > Set GravityOverride(GravityOverride=vector + vector(A=@n11, B=Get add Gravity JumpSide() -> (add Gravity JumpSide:Vector)))
   n12 = Get bIsSkyDiving()
   n13 = Get bIsJumping()
   n14 = Get bIsFalling()
   n15 = Branch(execute=@n10.then_0, Condition=AND Boolean(A=NOT Boolean(A=@n12), B=OR Boolean(A=@n13, B=@n14)))
   n16 = GetOwnerDirection!(self:Custom Character Part Anim Instance Object Reference) -> (ReturnValue:float)
   n17 = "Absolute (Float)"(A=float - float(A="Absolute (Float)"(A=@n16), B=90))
   > Set!(
    execute
    add Gravity JumpSide_X:double=0
    add Gravity JumpSide_Y:double=0
    add Gravity JumpSide_Z:double=float * float(A=MapRangeClamped(Value=@n17, InRangeA=0, InRangeB=45, OutRangeA=1, OutRangeB=0), B=500) {is=title K2Node_PromotableOperator}
   ) -> (then, Output_Get:Vector)
   Set add Gravity JumpSide(execute=@n15.else, add Gravity JumpSide:Vector="0, 0, 0") -> (Output_Get:Vector)
  ]}
]
variables [
 add Gravity JumpSide {type=Vector}
 Body_Mesh {type=SkeletalMeshComponent}
 Current LOD {type=integer}
 FE_idel {type=boolean}
 Gravity Override Params {type=GravityOverrideParamsStruct}
 GravityOverride {type=Vector}
 Is Backbling Cape {type=boolean}
 Is Moving 2D {type=boolean}
 Is Targeting {type=boolean}
 Is Wearing Backbling {type=boolean}
 scale_extra_head {type=boolean}
 Skel Mesh {type=SkeletalMeshComponent}
 skydive {type=boolean}
 SPR_HeadUp {type=number}
]
